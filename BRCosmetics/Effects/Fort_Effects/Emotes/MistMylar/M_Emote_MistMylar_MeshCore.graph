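# Translucent unlit Niagara material for the mesh core of the MistMylar emote effect, with emissive from particle color through MF_FX_EmissiveAdjustments.
# Opacity combines particle alpha, inverted Fresnel and fades at both U ends, then DepthFade softens intersections.
# Two blended panning noise textures, weighted along U, displace vertices along normals, with dynamic parameters controlling UV offset, scale and amplitude.
# A smoke-tinted color lerp and a second Fresnel feed nothing visible.
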Material /BRCosmetics/Effects/Fort_Effects/Emotes/MistMylar/M_Emote_MistMylar_MeshCore
graph {
 properties {
  bAutomaticallySetUsageInEditor=false
  bUsedWithNiagaraMeshParticles=true
  bUsedWithNiagaraRibbons=true
  bUsedWithNiagaraSprites=true
  blendMode=BLEND_Translucent
  ditherOpacityMask=true
  shadingModel=MSM_Unlit
  usedShadingModels=Unlit
 }
 nodes [
  n1 = ParticleColor()
  n2 = Multiply(A=MaterialFunctionCall(Invert Fresnel=StaticBool() {value=true}, Power=ScalarParameter() {defaultValue=4, parameterName="Power (S)"}) {materialFunction=&Fresnel_Function}.Result) {constB=2}
  n3 = TextureCoordinate()
  n4 = Power(Base=Multiply(A=ComponentMask(None=@n3) {r=true})) {constExponent=4}
  n5 = Multiply(A=Multiply(A=@n1.A, B=Saturate(None=@n2)), B=Saturate(None=OneMinus(None=@n4)))
  n6 = ComponentMask(None=TextureCoordinate()) {r=true}
  n7 = Saturate(None=Multiply(A=Power(Base=@n6) {constExponent=1}) {constB=3})
  DepthFade(Opacity=Multiply(A=Multiply(A=@n5, B=@n7))) {fadeDistanceDefault=50}
  Fresnel(
   ExponentIn=ScalarParameter() {defaultValue=1, parameterName=Fresnel_Exp}
   BaseReflectFractionIn=ScalarParameter() {parameterName=Fresnel_BRF}
  )
  n8 = Panner(Coordinate=TextureCoordinate() {uTiling=0.5}) {speedX=-1}
  LinearInterpolate(
   A=VectorParameter() {parameterName=Color}.RGB
   B=LinearInterpolate(Alpha=TextureSample(UVs=@n8) {texture=&T_Inky_Smoke_Tile}.R) {constB=0.5}
  ) {constAlpha=0.2}
  MaterialFunctionCall(Emissive=@n1.RGB) {materialFunction=&MF_FX_EmissiveAdjustments}
  n9 = VertexNormalWS()
  n10 = TextureCoordinate()
  n11 = DynamicParameter() {defaultValue=<rgba 1 1 80 1>, paramNames [Offset,Uscale,WPO,Param4]}
  n12 = AppendVector(A=@n11.Offset, B=Constant())
  n13 = Constant() {r=1}
  n14 = Panner(
   Coordinate=Multiply(A=Add(A=@n10, B=@n12), B=AppendVector(A=@n11.Uscale, B=@n13))
   Speed=Multiply(A=Constant2Vector() {r=-1}, B=ScalarParameter() {defaultValue=0.7, group=WPO, parameterName=WPOPanner_Speed})
  ) {speedX=-2}
  n15 = TextureCoordinate()
  n16 = Multiply(
   A=LinearInterpolate(
    A=TextureSample(UVs=@n14) {samplerType=SAMPLERTYPE_Grayscale, texture=&LowResBlurredNoise}.R
    B=TextureSample(UVs=@n14) {texture=/BRCosmetics/Weapons/FORT_Melee/Pickaxe_CoatCheck/FX/T_Pickaxe_CoatCheck_Noise}.R
   ) {constAlpha=0.2}
   B=LinearInterpolate(A=ScalarParameter() {defaultValue=0.05, group=WPO, parameterName=WPOMaskLerp}, Alpha=ComponentMask(None=@n15) {r=true}) {constA=0.05}
  )
  Reroute(None=Multiply(A=Multiply(A=@n9, B=@n16), B=DynamicParameter() {defaultValue=<rgba 1 1 80 1>, paramNames [Offset,Uscale,WPO,Param4]}.WPO))
  StaticBool() {value=true}
 ]
}
