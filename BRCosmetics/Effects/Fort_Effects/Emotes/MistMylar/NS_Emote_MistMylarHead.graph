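# Niagara system for the MistMylar emote, named for the head, that produces a short puff of dark, rising smoke.
# Its single SmokeLocal sprite emitter spawns particles in a small offset sphere with upward and cone velocity, slowed by drag.
# A region mask, moved down along Z over emitter age, scales the alpha fade curve.
# Sprites rotate, change size, use SubUV frames and random material parameters, and take near-black tint from time-of-day colors.
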NiagaraSystem /BRCosmetics/Effects/Fort_Effects/Emotes/MistMylar/NS_Emote_MistMylarHead
dependencies {
 usedDataInterfaces [
  &NiagaraDataInterfaceCurve
  &NiagaraDataInterfaceColorCurve
  &NiagaraDataInterfaceSpriteRendererInfo
 ]
 usedDynamicInputs []
 usedModules [&SystemState]
}
emitters {SmokeLocal {
  emitterUpdateScript [
   &EmitterState(Life Cycle Mode=::NewEnumerator1, Loop Behavior=::NewEnumerator1, Loop Delay=0.4, Loop Duration=1.5, UseLoopDelay=false)
   &SpawnRate(SpawnRate=30, Use Spawn Probability=true)
  ]
  particleSpawnScript [
   &PartitionParticles(Number Of Partitions=1)
   &InitializeParticle(
    Color=&MultiplyLinearColor(Linear Color=<rgba 0.7031 0.7031 0.7031 0.8>, Scale Factor=&MakeVector4_FromFloat(Float=1))
    Lifetime Max=1
    Lifetime Mode=::NewEnumerator1
    Position Offset=<0 -10 -50>
    Sprite Rotation Mode=::NewEnumerator2
    Sprite Size Mode=::NewEnumerator1
    Uniform Sprite Size Max=80
    Uniform Sprite Size Min=100
    UsePositionOffset=true
    Write Material Random=true
   )
   &AddVelocity(
    Scale Added Velocity=&Multiply_VectorByFloat(
     Vector=&Multiply_VectorByFloat(Vector=<2.5 2.5 2.5>, Float=1)
     Float=&Lerp_Float(A=1, B=0.6, Alpha=$System.InterpolateOverTime.InterpolateOverTime_Float.Moving Average)
    )
    Velocity=<0 0 50>
   )
   &AddVelocityInCone(Cone Angle=50, Cone Axis=<0 0 1>, Module.Use Velocity Falloff On Cone Axis=false, Velocity Strength=100) {module=AddVelocityInCone001}
   &CameraOffset()
   &ShapeLocation(MaskSpawnGroup=false, Override Seed=false, Sphere Radius=20, UseOwnerScale=true)
  ]
  particleUpdateScript [
   &Drag(Use Linear Drag=true, Use Rotational Drag=true)
   &SolveForcesAndVelocity(Clamp Velocity=false, Limit Acceleration=false)
   &ParticleState()
   &SpriteRotationRate(Rotation Rate=50)
   &ScaleSpriteSize(
    Uniform Curve Scale=0.7
    Uniform Curve Sprite Scale=DataInterface() {propertyValues {Curve {keys [
        {arriveTangent=10.8988, interpMode=RCIM_Cubic, leaveTangent=10.8988, tangentMode=RCTM_User, time=0, value=1}
        {arriveTangent=-1.0121, interpMode=RCIM_Cubic, leaveTangent=-1.0121, tangentMode=RCTM_User, time=0.3, value=1.175}
        <at 1 0.5>
       ]}}}
   )
   &RegionMask(
    Offset:Vector3f=&MakeVector(
     X=0
     Y=0
     Z=&FloatFromCurve(FloatCurve=DataInterface() {propertyValues {Curve {keys [<at 0 200>,<at 1 0>]}}}, CurveIndex=$Emitter.Age)
    )
    Offset Space:ENiagaraCoordinateSpace=::Local
    Region Dimensions:Vector3f=<800 800 800>
    Region Type:ENiagaraDebugDrawShape=::NewEnumerator1
    UseOffset=true
    UseRotation=false
   )
   &ScaleColor(
    Scale Alpha=&FloatFromCurve(
     FloatCurve=DataInterface() {propertyValues {Curve {keys [
         {arriveTangent=14.7867, interpMode=RCIM_Cubic, leaveTangent=14.7867, tangentMode=RCTM_User, time=0, value=0}
         {arriveTangent=-2.3483, interpMode=RCIM_Cubic, leaveTangent=-2.3483, tangentMode=RCTM_User, time=0.3, value=1}
         <at 0.6 0.15>
         <at 1 0>
        ]}}}
     Scale Curve=&Multiply_Float(A=2, B=$Particles.RegionMask.Float)
    )
    Scale RGB=&MakeVectorFromLinearColorRGB(Color=&FNTimeOfDayColor(
      Night Color=&ColorFromCurve(ColorCurve=DataInterface() {propertyValues {
         AlphaCurve {keys []}
         BlueCurve {keys [<at 0.7598 0.01 RCIM_Cubic>]}
         GreenCurve {keys [<at 0.7598 0.0055 RCIM_Cubic>]}
         RedCurve {keys [<at 0.7598 0.004 RCIM_Cubic>]}
        }})
      Morning Color=&ColorFromCurve(ColorCurve=DataInterface() {propertyValues {
         AlphaCurve {keys []}
         BlueCurve {keys [<at 0.7598 0.01 RCIM_Cubic>]}
         GreenCurve {keys [<at 0.7598 0.01 RCIM_Cubic>]}
         RedCurve {keys [<at 0.7598 0.01 RCIM_Cubic>]}
        }})
      Day Color=&ColorFromCurve(ColorCurve=DataInterface() {propertyValues {
         AlphaCurve {keys []}
         BlueCurve {keys [<at 0.7598 0.01 RCIM_Cubic>]}
         GreenCurve {keys [<at 0.7598 0.01 RCIM_Cubic>]}
         RedCurve {keys [<at 0.7598 0.01 RCIM_Cubic>]}
        }})
      Evening Color=&ColorFromCurve(ColorCurve=DataInterface() {propertyValues {
         AlphaCurve {keys []}
         BlueCurve {keys [<at 0.7598 0.01 RCIM_Cubic>]}
         GreenCurve {keys [<at 0.7598 0.01 RCIM_Cubic>]}
         RedCurve {keys [<at 0.7598 0.01 RCIM_Cubic>]}
        }})
     ))
    ScaleA=true
    ScaleRGB=true
   )
   &SubUVAnimation(
    End Frame Range Override=3
    Evaluation Type=::NewEnumerator0
    "Sprite Renderer "=DataInterface() {propertyValues {SpriteRenderer=&self:SmokeLocal_7.NiagaraSpriteRendererProperties_1}}
    SubUV Animation Mode:ENiagara_SubUVLookupModeV2=::NewEnumerator2
    UseEndFrame=true
    UseStartFrame=true
   ) {module=SubUVAnimation001}
   &DynamicMaterialParameters(
    Index 0 Param 1=&RandomRangeFloat(Minimum=0, Maximum=1)
    Module.Index0Param0Write=true
    Module.Index0Param1Write=true
    Module.Index0Param2Write=true
    Module.Index0Param3Write=true
   ) {module=DynamicMaterialParameters001}
  ]
  rendererClasses [&NiagaraSpriteRendererProperties]
  renderers [{rendererClass=&NiagaraSpriteRendererProperties}]
 }}
summary {
 systemName=NS_Emote_MistMylarHead
 emitters [{emitterName=SmokeLocal, rendererClasses [&NiagaraSpriteRendererProperties]}]
 userVariables [
  User.bIsFrontEnd {type=NiagaraBool}
  User.bIsFrontEndPreview {type=NiagaraBool}
 ]
}
