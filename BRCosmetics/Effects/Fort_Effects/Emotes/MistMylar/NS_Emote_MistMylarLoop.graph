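# Looping Niagara effect for the MistMylar emote, wrapping the character in dark smoky mist with trails.
# Core meshes, mask trails, spiral meshes and ribbons follow skeletal bones or sockets, alongside cloud meshes and flipbook smoke.
# Trail emitters smooth particle positions over time for use as dynamic data, while ribbons respond to owner speed.
# Age-driven region masks fade parts, and colors or rates vary by front-end preview, time of day and platform.
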NiagaraSystem /BRCosmetics/Effects/Fort_Effects/Emotes/MistMylar/NS_Emote_MistMylarLoop
dependencies {
 usedDataInterfaces [
  &NiagaraDataInterfacePlatformSet
  &NiagaraDataInterfaceArrayFloat
  &NiagaraDataInterfaceArrayInt32
  &NiagaraDataInterfaceArrayFloat3
  &NiagaraDataInterfaceCurve
  &NiagaraDataInterfaceColorCurve
  &NiagaraDataInterfaceSpriteRendererInfo
  &NiagaraDataInterfaceSkeletalMesh
 ]
 usedDynamicInputs []
 usedModules [
  &self:SystemSpawnScript.SystemScriptSource.SystemScriptGraph.NiagaraNodeAssignment_0.SetVariables_#1
  &self:SystemSpawnScript.SystemScriptSource.SystemScriptGraph.NiagaraNodeAssignment_1.SetVariables_#2
  &self:SystemSpawnScript.SystemScriptSource.SystemScriptGraph.NiagaraNodeAssignment_2.SetVariables_#3
  &SystemState
 ]
}
emitters {
 BodyCore {
  emitterUpdateScript [
   &EmitterState(
    Inactive Response=::NewEnumerator1
    Life Cycle Mode=::NewEnumerator1
    Loop Behavior=::NewEnumerator1
    Loop Duration Mode=::NewEnumerator1
    UseLoopDelay=false
   )
   &SpawnBurst_Instantaneous(Spawn Count=4, Use Spawn Probability=false, UseLoopCountLimit=false)
  ]
  particleSpawnScript [
   &PartitionParticles(Number Of Partitions=4)
   &InitializeParticle(
    Color=&Multiply_LinearColorByFloat(
     Linear Color=<rgba 1 1 1 1>
     Float=&FN_PickFloatDependOnPlatform(TargetPlatformActive=1, TargetPlatformNoActive=1.5)
    )
    Mesh Scale Mode=::NewEnumerator3
    Mesh Uniform Scale=&SelectFloatFromArray(
     Direct Array Index=$Particles.PartitionParticles.Partition
     Float Selection Array=DataInterface() {propertyValues {FloatData [0.2,0.1,0.1,0.2]}}
    )
    Ribbon Width=10
    Ribbon Width Mode=::NewEnumerator1
    UsePositionOffset=false
    Write Material Random=true
   )
   &InitialMeshOrientation(Mesh Orientation Mode=::NewEnumerator2, Rotation=<0.25 0 0>, Use Rotation Vector=true)
   &InitialMeshOrientation(Mesh Orientation Mode=::NewEnumerator2, Use Rotation Vector=true) {module=InitialMeshOrientation001}
  ]
  particleUpdateScript [
   &SkeletalMeshLocation(
    Bone / Socket Index=&SelectIntFromArray(Direct Array Index=$Particles.PartitionParticles.Partition, Int Selection Array=DataInterface() {propertyValues {IntData [6,7,9,10,11]}})
    Bone Sampling Mode=::NewEnumerator4
    MaskSpawnGroup=false
    Sampled Position Offset=&SelectVectorFromArray(
     Direct Array Index=$Particles.PartitionParticles.Partition
     Vector Selection Array=DataInterface() {propertyValues {FloatData [<0 0 0>, <0 0 -10>, <0 0 -40>, <0 0 -20>]}}
    )
    Skeletal Mesh=$System.Skeletal Mesh
   )
   &SolveForcesAndVelocity(Clamp Velocity=false, Limit Acceleration=false)
   &ParticleState(Kill Particles When Lifetime Has Elapsed=false)
   &ScaleMeshSize(Scale Factor=&Multiply_VectorByFloat(Vector=<1 1 1>, Float=5))
   &DynamicMaterialParameters(
    Index 0 Param 1=&SelectFloatFromArray(
     Direct Array Index=$Particles.PartitionParticles.Partition
     Float Selection Array=DataInterface() {propertyValues {FloatData [0.2,0.4,0.7,0.5]}}
    )
    Index 0 Param 2=&SelectFloatFromArray(
     Direct Array Index=$Particles.PartitionParticles.Partition
     Float Selection Array=DataInterface() {propertyValues {FloatData [0.3,0.5,0.8,0.5]}}
    )
    Index 0 Param 3=150
    Module.Index0Param0Write=true
    Module.Index0Param1Write=true
    Module.Index0Param2Write=true
    Module.Index0Param3Write=true
   )
   &RegionMask(
    Offset:Vector3f=&MakeVector(
     X=0
     Y=&FloatFromCurve(FloatCurve=DataInterface() {propertyValues {Curve {keys [<at 0 -300>,<at 0.7 0>]}}}, CurveIndex=$Emitter.Age)
     Z=0
    )
    Offset Space:ENiagaraCoordinateSpace=::Local
    Region Dimensions:Vector3f=<500 500 500>
    Region Falloff=1
    Region Type:ENiagaraDebugDrawShape=::NewEnumerator1
    UseOffset=true
    UseRotation=false
   )
   &ScaleColor(
    Scale Alpha=&Multiply_Float(A=$Particles.RegionMask.Float, B=1)
    Scale RGB=&MakeVectorFromLinearColorRGB(Color=&FNTimeOfDayColor(
      Night Color=<rgba 0.004 0.0055 0.01 1>
      Morning Color=<rgba 0.009 0.0098 0.01 1>
      Day Color=<rgba 0.009 0.0098 0.01 1>
      Evening Color=<rgba 0.009 0.0098 0.01 1>
     ))
    ScaleA=true
    ScaleRGB=true
   )
  ]
  rendererClasses [&NiagaraMeshRendererProperties]
  renderers [{rendererClass=&NiagaraMeshRendererProperties, rendererIndex=0}]
 }
 Cloud {
  emitterUpdateScript [
   &EmitterState(
    Delay First Loop Only=true
    Inactive Response=::NewEnumerator1
    Life Cycle Mode=::NewEnumerator1
    Loop Delay=0.6
    Loop Duration=1
    UseLoopDelay=true
   )
   &SpawnBurst_Instantaneous(Spawn Count=5, Use Spawn Probability=false, UseLoopCountLimit=false)
   &SpawnRate(SpawnRate=&FN_FrontEndFloatModifier(Gameplay=5, Front End=10), Use Spawn Probability=true)
   /SetVariables(Emitter.FEdes:NiagaraFloat=&FN_FrontEndFloatModifier(Gameplay=0, Front End=0)) {isSetParameters=true}
   /SetVariables(Emitter.ColorCenter:LinearColor=&MakeCustomLinearColorFromBool(Bool=$User.bIsFrontEndPreview, True Color=<rgba 0.0097 0.0131 0.02 1>, False Color=<rgba 0.0225 0.0245 0.025 1>)) {isSetParameters=true}
   /SetVariables(Emitter.ColorOutSide:LinearColor=&MakeCustomLinearColorFromBool(Bool=$User.bIsFrontEndPreview, True Color=<rgba 0.0117 0.0185 0.02 1>, False Color=<rgba 0.0237 0.0256 0.026 1>)) {isSetParameters=true}
   /SetVariables(Emitter.ColorNoise:LinearColor=&Multiply_LinearColorByFloat(
     Linear Color=&MakeCustomLinearColorFromBool(Bool=$User.bIsFrontEndPreview, True Color=<rgba 0.0234 0.0287 0.03 1>, False Color=<rgba 0.0546 0.0589 0.06 1>)
     Float=&FN_PickFloatDependOnPlatform(TargetPlatformActive=1, TargetPlatformNoActive=1)
    )) {isSetParameters=true}
  ]
  particleSpawnScript [
   &PartitionParticles(Number Of Partitions=1)
   &InitializeParticle(
    Color=&MultiplyLinearColor(
     Linear Color=<rgba 0.7031 0.7031 0.7031 0.8>
     Scale Factor=&MakeVector4_FromFloat(Float=&FN_FrontEndFloatModifier(Gameplay=1, Front End=1))
    )
    Lifetime Min=1.5
    Lifetime Mode=::NewEnumerator1
    Mesh Scale=<1.35 0.9 0.9>
    Mesh Scale Mode=::NewEnumerator4
    Position Offset=<0 0 30>
    UsePositionOffset=true
    Write Material Random=true
   )
   &AddVelocity(
    Scale Added Velocity=&Multiply_VectorByFloat(Vector=&Multiply_VectorByFloat(Vector=<2.5 2.5 2.5>, Float=1), Float=1)
    Velocity=<0 0 -50>
   )
   &InitialMeshOrientation(
    Mesh Orientation Mode=::NewEnumerator2
    Rotation=&RandomRangeVector(Minimum=<0 0 0>, Maximum=<1 0 0>)
    Use Rotation Vector=true
   )
   &CameraOffset(Camera Offset Amount=-30)
   &ShapeLocation(MaskSpawnGroup=false, Override Seed=false, Sphere Radius=20, UseOwnerScale=true)
  ]
  particleUpdateScript [
   &Drag(Drag=0, Use Linear Drag=true, Use Rotational Drag=true)
   &SolveForcesAndVelocity(Clamp Velocity=false, Limit Acceleration=false)
   &ParticleState()
   &ScaleMeshSize(Scale Factor=&Multiply_VectorByFloat(
     Vector=&Multiply_VectorByFloat(Vector=<1 1 1>, Float=&RandomRangeFloat(Minimum=0.7, Maximum=1))
     Float=&FloatFromCurve(
      FloatCurve=DataInterface() {propertyValues {Curve {keys [
          {arriveTangent=11.15, interpMode=RCIM_Cubic, leaveTangent=11.15, tangentMode=RCTM_User, time=0, value=0}
          {arriveTangent=-1.0121, interpMode=RCIM_Cubic, leaveTangent=-1.0121, tangentMode=RCTM_User, time=0.2899, value=0.9265}
          <at 1 0.5>
         ]}}}
      Scale Curve=1.6
     )
    ))
   &DynamicMaterialParameters(
    Index 0 Param 1=&FloatFromCurve(
     FloatCurve=DataInterface() {propertyValues {Curve {keys [
         {interpMode=RCIM_Cubic, tangentMode=RCTM_User, time=0, value=0}
         <at 0.2 0.55>
         {arriveTangent=-1.0223, interpMode=RCIM_Cubic, leaveTangent=-1.0223, tangentMode=RCTM_User, time=1, value=1}
        ]}}}
     Scale Curve=&RandomRangeFloat(Minimum=3, Maximum=3)
    )
    Module.Index0Param0Write=true
    Module.Index0Param1Write=true
    Module.Index0Param2Write=true
    Module.Index0Param3Write=true
   )
   &RegionMask(
    Offset:Vector3f=&MakeVector(
     X=0
     Y=0
     Z=&FloatFromCurve(FloatCurve=DataInterface() {propertyValues {Curve {keys [<at 0 200>,<at 1 0>]}}}, CurveIndex=$Emitter.Age)
    )
    Offset Space:ENiagaraCoordinateSpace=::Local
    Region Dimensions:Vector3f=<500 500 500>
    Region Type:ENiagaraDebugDrawShape=::NewEnumerator1
    UseOffset=true
    UseRotation=false
   )
   &ScaleColor(
    Scale Alpha=&FloatFromCurve(
     FloatCurve=DataInterface() {propertyValues {Curve {keys [
         {arriveTangent=10.2356, interpMode=RCIM_Cubic, leaveTangent=10.2356, tangentMode=RCTM_User, time=0, value=0}
         <at 0.3 1>
         <at 0.6 0.1>
         <at 1 0>
        ]}}}
     Scale Curve=&Multiply_Float(A=&FN_FrontEndFloatModifier(Gameplay=1.5, Front End=1.5), B=$Particles.RegionMask.Float)
    )
    Scale RGB=&Multiply_Vector(
     A=&MakeVectorFromLinearColorRGB(Color=&ColorFromCurve(ColorCurve=DataInterface() {propertyValues {
         AlphaCurve {keys []}
         BlueCurve {keys [{arriveTangent=-331.6383, interpMode=RCIM_Cubic, time=0, value=0.0781}, <at 0.7598 0 RCIM_Cubic>]}
         GreenCurve {keys [{arriveTangent=-87.3695, interpMode=RCIM_Cubic, time=0, value=0.0781}, <at 0.7598 0 RCIM_Cubic>]}
         RedCurve {keys [<at 0 0.0781 RCIM_Cubic>,<at 0.7598 0 RCIM_Cubic>]}
        }}))
     B=&Multiply_VectorByFloat(Vector=<1 1 1>, Float=1)
    )
    ScaleA=true
    ScaleRGB=true
   )
  ]
  rendererClasses [&NiagaraMeshRendererProperties]
  renderers [{rendererClass=&NiagaraMeshRendererProperties, rendererIndex=0}]
 }
 MaskTrails {
  emitterUpdateScript [
   &EmitterState(
    Inactive Response=::NewEnumerator1
    Life Cycle Mode=::NewEnumerator1
    Loop Behavior=::NewEnumerator1
    Loop Duration Mode=::NewEnumerator1
    UseLoopDelay=false
   )
   &SpawnBurst_Instantaneous(Spawn Count=5, Use Spawn Probability=false, UseLoopCountLimit=false)
  ]
  particleSpawnScript [
   &PartitionParticles(Number Of Partitions=5)
   &InitializeParticle(
    Color Maximum=<rgba 0.01 0.01 0.01 3>
    Color Minimum=<rgba 0.01 0.01 0.01 2>
    Color Mode=::NewEnumerator2
    Mesh Scale=&SelectVectorFromArray(
     Direct Array Index=$Particles.PartitionParticles.Partition
     Vector Selection Array=DataInterface() {propertyValues {FloatData [<1 2.2 2>, <1 1.8 1.5>, <1 1.3 1>, <1 1.5 1>, <1 1.8 1>, <1 1.5 1>, <1 1.2 1>, <1 1.3 1>]}}
    )
    Mesh Scale Mode=::NewEnumerator4
    UsePositionOffset=true
    Write Material Random=true
   )
   &InitialMeshOrientation(
    Mesh Orientation Mode=::NewEnumerator2
    Rotation=&FN_FrontEndVector3Modifier(Gameplay=<-0.05 0 0>, Front End=<-0.05 0 0>)
    Use Rotation Vector=true
   )
   &InitialMeshOrientation(
    Mesh Orientation Mode=::NewEnumerator2
    Rotation=&SelectVectorFromArray(
     Direct Array Index=$Particles.PartitionParticles.Partition
     Vector Selection Array=DataInterface() {propertyValues {FloatData [<0 -0.08 0>, <0 0.1 0>, <0 -0.08 0>, <0.05 0.07 0>, <0.03 -0.05 0>, <0 -0.05 0>, <0.03 0.05 0>, <-0.03 -0.03 0>]}}
    )
    Use Rotation Vector=true
   ) {module=InitialMeshOrientation001}
   &CameraOffset(Camera Offset Amount=-5)
  ]
  particleUpdateScript [
   &ParticleState(Kill Particles When Lifetime Has Elapsed=false, Let Infinitely Lived Particles Die When Emitter Deactivates=true)
   &ScaleMeshSize(Scale Factor=&Multiply_VectorByFloat(Vector=<2 0.6 2>, Float=0.3))
   &SkeletalMeshLocation(
    Bone / Socket Index=$Particles.PartitionParticles.Partition
    Bone Sampling Mode=::NewEnumerator4
    MaskSpawnGroup=false
    Sampled Position Offset=&AddVectorToPosition(
     Position=&SelectVectorFromArray(
      Direct Array Index=$Particles.PartitionParticles.Partition
      Vector Selection Array=DataInterface() {propertyValues {FloatData [<0 0 0>, <0 0 0>, <0 0 0>, <0 0 0>, <0 0 0>]}}
     )
     Vector=<0 -10 0>
    )
    Skeletal Mesh=$System.Skeletal Mesh
   )
   &DynamicMaterialParameters(
    Index 0 Param 1=&RandomRangeFloat(Minimum=0.2, Maximum=1)
    Index 0 Param 2=&RandomRangeFloat(Minimum=0.2, Maximum=1)
    Index 0 Param 3=0.3
    Index 0 Param 4=20
    Index 3 Param 1=150
    Index 3 Param 4=&FloatFromCurve(FloatCurve=DataInterface() {propertyValues {Curve {keys [<at 0 0>,<at 1 1>]}}})
    Module.Index0Param0Write=true
    Module.Index0Param1Write=true
    Module.Index0Param2Write=true
    Module.Index0Param3Write=true
    Module.Index3Param0Write=true
    Module.Index3Param1Write=true
    Module.Index3Param2Write=true
    Module.Index3Param3Write=true
    Write Parameter Index 3=true
   )
   &InterpolateOverTime(
    Rate Of Change:NiagaraFloat=&Multiply_Float(A=3, B=$System.InterpolateOverTime.InterpolateOverTime_Float.Moving Average)
    Target Vector Value:NiagaraPosition=$Particles.Position
    Variable Type=::NewEnumerator2
   )
   /SetVariables(Particles.Interpolate Dynamic=&MakeFloat4FromVectorAndFloat("Vector (RGB)"=$StackContext.InterpolateOverTime.InterpolateOverTime_V3_.Moving Average, "Float (Alpha)"=3)) {isSetParameters=true, module=SetVariables_#4}
   /SetVariables(Particles.Color002=&MakeCustomLinearColorFromBool(
     Bool=$User.bIsFrontEndPreview
     True Color=<rgba 0.0114 0.0152 0.02 0.6>
     False Color=&FNTimeOfDayColor(
      Night Color=<rgba 0.009 0.014 0.03 0.6>
      Morning Color=<rgba 0.03 0.03 0.03 0.6>
      Day Color=<rgba 0.03 0.03 0.03 0.6>
      Evening Color=<rgba 0.03 0.03 0.03 0.6>
     )
    )) {isSetParameters=true, module=SetVariables_#5}
   &ScaleColor(
    Scale Alpha=&FloatFromCurve(FloatCurve=DataInterface() {propertyValues {Curve {keys [<at 0 0>,<at 0.1 1>]}}})
    Scale RGB=&MakeCustomVectorFromBool(
     Bool=$User.bIsFrontEndPreview
     True Vector=&MakeVectorFromLinearColorRGB(Color=<rgba 0.4121 0.7086 1 1>)
     False Vector=&MakeVectorFromLinearColorRGB(Color=&FNTimeOfDayColor(Night Color=<rgba 0.3 0.475 1 1>, Morning Color=<rgba 1 1 1 1>, Day Color=<rgba 1 1 1 1>, Evening Color=<rgba 1 1 1 1>))
    )
    ScaleA=true
    ScaleRGB=true
   )
   &FN_Sprite_Attributes(Create Color Attribute=true, Write to Module Namespace=true)
  ]
  rendererClasses [&NiagaraMeshRendererProperties]
  renderers [
   {rendererClass=&NiagaraMeshRendererProperties, rendererIndex=0}
   {rendererClass=&NiagaraMeshRendererProperties, rendererIndex=1}
  ]
 }
 RibbonTail {
  emitterUpdateScript [
   &EmitterState(
    Delay First Loop Only=true
    Inactive Response=::NewEnumerator1
    Life Cycle Mode=::NewEnumerator1
    Loop Delay=0.5
    Loop Duration=1
    UseLoopDelay=true
   )
   &SpawnPerUnit(
    Module.Use Max Movement Threshold=true
    Module.Use Movement Tolerance=false
    Movement Tolerance=0
    Spawn Probability:NiagaraFloat=&MakeCustomFloatFromBool(
     Bool=&SetBoolByFloatComparison(A=&VectorLength(Vector=$Engine.Owner.Velocity), B=500)
     True Float=0
     False Float=1
    )
    Spawn Spacing=1
    Use Spawn Probability=true
   )
  ]
  particleSpawnScript [
   &PartitionParticles(Number Of Partitions=5)
   &InitializeParticle(
    Color=<rgba 0 0 0 1>
    Lifetime=0.6
    Ribbon Width=60
    Ribbon Width Mode=::NewEnumerator1
    Sprite Size Mode=::NewEnumerator3
    Uniform Sprite Size=20
    UsePositionOffset=false
    Write Material Random=true
   )
   &AddVelocity(Velocity=<0 0 -200>)
   &SkeletalMeshLocation(
    Bone / Socket Index=$Particles.PartitionParticles.Partition
    MaskSpawnGroup=false
    Mesh Sampling Type=::NewEnumerator5
    Sampled Position Offset=<0 0 20>
    Skeletal Mesh=$System.Skeletal Mesh
    Socket Sampling Mode=::NewEnumerator5
   )
   &DynamicMaterialParameters(
    Index 0 Param 1=&SelectFloatFromArray(
     Direct Array Index=$Particles.PartitionParticles.Partition
     Float Selection Array=DataInterface() {propertyValues {FloatData [1,1,0.5,0.2,0]}}
    )
    Index 0 Param 2=1
    Index 0 Param 4=20
    Index 3 Param 1=150
    Module.Index0Param0Write=true
    Module.Index0Param1Write=true
    Module.Index0Param2Write=true
    Module.Index0Param3Write=true
    Module.Index3Param0Write=true
    Module.Index3Param1Write=true
    Module.Index3Param2Write=true
    Module.Index3Param3Write=true
    Write Parameter Index 3=true
   )
   /SetVariables(Particles.DY2:Vector4f=&MakeVector4(
     X=-0.7
     Y=0
     Z=&SelectFloatFromArray(
      Direct Array Index=$Particles.PartitionParticles.Partition
      Float Selection Array=DataInterface() {propertyValues {FloatData [0.2,0.4,0.5,0.2,0]}}
     )
     W=1
    )) {isSetParameters=true}
  ]
  particleUpdateScript [
   &InterpolateOverTime(
    Rate Of Change=5
    Target Float Value:NiagaraFloat=&MakeCustomFloatFromBool(
     Bool=&SetBoolByFloatComparison(A=&VectorLength(Vector=$Engine.Owner.Velocity), B=100)
     True Float=1.3
     False Float=0
    )
   )
   /SetVariables(Particles.Color002=&MakeLinearColorFromVectorAndFloat(
     "Vector (RGB)"=&MakeVectorFromLinearColorRGB(Color=&FNTimeOfDayColor(
       Night Color=<rgba 0.009 0.0142 0.03 0.7>
       Morning Color=<rgba 0.03 0.03 0.03 0.7>
       Day Color=<rgba 0.03 0.03 0.03 0.7>
       Evening Color=<rgba 0.03 0.03 0.03 0.7>
      ))
     "Float (Alpha)"=&Multiply_Float(A=0.7, B=$StackContext.InterpolateOverTime.InterpolateOverTime_Float.Moving Average)
    )) {isSetParameters=true}
   &SolveForcesAndVelocity(Clamp Velocity=false, Limit Acceleration=false)
   &ParticleState()
   &ScaleRibbonWidth()
   &ScaleColor(Scale Alpha=$StackContext.InterpolateOverTime.InterpolateOverTime_Float.Moving Average, ScaleA=true, ScaleRGB=true)
  ]
  rendererClasses [&NiagaraRibbonRendererProperties]
  renderers [
   {rendererClass=&NiagaraRibbonRendererProperties}
   {rendererClass=&NiagaraRibbonRendererProperties, rendererIndex=1}
  ]
 }
 SmokeLocal {
  emitterUpdateScript [
   &EmitterState(
    Delay First Loop Only=true
    Inactive Response=::NewEnumerator1
    Life Cycle Mode=::NewEnumerator1
    Loop Delay=0.4
    Loop Duration=1
    UseLoopDelay=true
   )
   &SpawnBurst_Instantaneous(
    Spawn Count=5
    Spawn Probability=&FN_FrontEndFloatModifier(Gameplay=1, Front End=0.5)
    Use Spawn Probability=true
    UseLoopCountLimit=false
   )
   &SpawnRate(
    Spawn Probability=&FN_FrontEndFloatModifier(Gameplay=1, Front End=0.5)
    SpawnRate=20
    Use Spawn Probability=true
   )
  ]
  particleSpawnScript [
   &PartitionParticles(Number Of Partitions=1)
   &InitializeParticle(
    Color=&MultiplyLinearColor(Linear Color=<rgba 0.7031 0.7031 0.7031 0.8>, Scale Factor=&MakeVector4_FromFloat(Float=1))
    Lifetime Min=1.5
    Lifetime Mode=::NewEnumerator1
    Position Offset=<0 0 50>
    Sprite Rotation Mode=::NewEnumerator2
    Sprite Size Mode=::NewEnumerator1
    Uniform Sprite Size Max=150
    Uniform Sprite Size Min=100
    UsePositionOffset=true
    Write Material Random=true
   )
   &AddVelocity(
    Scale Added Velocity=&Multiply_VectorByFloat(Vector=&Multiply_VectorByFloat(Vector=<2.5 2.5 2.5>, Float=1), Float=1)
    Velocity=<0 0 -50>
   )
   &AddVelocityInCone(Cone Angle=50, Cone Axis=<0 0 -1>, Module.Use Velocity Falloff On Cone Axis=false, Velocity Strength=100) {module=AddVelocityInCone001}
   &CameraOffset()
   &ShapeLocation(MaskSpawnGroup=false, Override Seed=false, Sphere Radius=30, UseOwnerScale=true)
  ]
  particleUpdateScript [
   &Drag(Drag=0.3, Use Linear Drag=true, Use Rotational Drag=true)
   &SolveForcesAndVelocity(Clamp Velocity=false, Limit Acceleration=false)
   &ParticleState()
   &SpriteRotationRate(Rotation Rate=50)
   &ScaleSpriteSize(Uniform Curve Sprite Scale=DataInterface() {propertyValues {Curve {keys [
        {arriveTangent=10.8988, interpMode=RCIM_Cubic, leaveTangent=10.8988, tangentMode=RCTM_User, time=0, value=0}
        {arriveTangent=-1.0121, interpMode=RCIM_Cubic, leaveTangent=-1.0121, tangentMode=RCTM_User, time=0.2899, value=0.9265}
        <at 1 0.5>
       ]}}})
   &RegionMask(
    Offset:Vector3f=&MakeVector(
     X=0
     Y=0
     Z=&FloatFromCurve(FloatCurve=DataInterface() {propertyValues {Curve {keys [<at 0 200>,<at 1 0>]}}}, CurveIndex=$Emitter.Age)
    )
    Offset Space:ENiagaraCoordinateSpace=::Local
    Region Dimensions:Vector3f=<800 800 800>
    Region Type:ENiagaraDebugDrawShape=::NewEnumerator1
    UseOffset=true
    UseRotation=false
   )
   &DynamicMaterialParameters(
    Index 0 Param 1=&FloatFromCurve(
     FloatCurve=DataInterface() {propertyValues {Curve {keys [
         {interpMode=RCIM_Cubic, tangentMode=RCTM_User, time=0, value=0}
         <at 0.2 0.55>
         {arriveTangent=-1.0223, interpMode=RCIM_Cubic, leaveTangent=-1.0223, tangentMode=RCTM_User, time=1, value=1}
        ]}}}
     Scale Curve=&RandomRangeFloat(Minimum=3, Maximum=3)
    )
    Module.Index0Param0Write=true
    Module.Index0Param1Write=true
    Module.Index0Param2Write=true
    Module.Index0Param3Write=true
   )
   &ScaleColor(
    Scale Alpha=&FloatFromCurve(
     FloatCurve=DataInterface() {propertyValues {Curve {keys [
         {arriveTangent=10.2356, interpMode=RCIM_Cubic, leaveTangent=10.2356, tangentMode=RCTM_User, time=0, value=0}
         <at 0.3 1>
         <at 0.7 0.4>
         <at 1 0>
        ]}}}
     Scale Curve=&Multiply_Float(A=0.6, B=$Particles.RegionMask.Float)
    )
    Scale RGB=&MakeVectorFromLinearColorRGB(Color=&ColorFromCurve(ColorCurve=DataInterface() {propertyValues {
        AlphaCurve {keys []}
        BlueCurve {keys [{arriveTangent=-331.6383, interpMode=RCIM_Cubic, time=0, value=0.0781}, <at 0.7598 0 RCIM_Cubic>]}
        GreenCurve {keys [{arriveTangent=-87.3695, interpMode=RCIM_Cubic, time=0, value=0.0781}, <at 0.7598 0 RCIM_Cubic>]}
        RedCurve {keys [<at 0 0.0781 RCIM_Cubic>,<at 0.7598 0 RCIM_Cubic>]}
       }}))
    ScaleA=true
    ScaleRGB=true
   )
   &SubUVAnimation(
    "Sprite Renderer "=DataInterface() {propertyValues {SpriteRenderer=&self:SmokeLocal_7.NiagaraSpriteRendererProperties_0}}
    UseEndFrame=true
    UseStartFrame=true
   )
  ]
  rendererClasses [&NiagaraSpriteRendererProperties]
  renderers [{rendererClass=&NiagaraSpriteRendererProperties}]
 }
 SmokeNoLocal {
  bEnabled=false
  emitterUpdateScript [
   &EmitterState(Delay First Loop Only=true, Life Cycle Mode=::NewEnumerator1, Loop Delay=0.3, Loop Duration=1, UseLoopDelay=true)
   &SpawnBurst_Instantaneous(
    Spawn Count=2
    Spawn Probability=&MakeCustomFloatFromBool(
     Bool=&SetBoolByFloatComparison(A=&VectorLength(Vector=$Engine.Owner.Velocity), B=500)
     True Float=0
     False Float=1
    )
    Use Spawn Probability=true
    UseLoopCountLimit=false
   )
   &SpawnRate(
    Spawn Probability=&MakeCustomFloatFromBool(
     Bool=&SetBoolByFloatComparison(A=&VectorLength(Vector=$Engine.Owner.Velocity), B=500)
     True Float=0
     False Float=1
    )
    SpawnRate=10
    Use Spawn Probability=true
   )
  ]
  particleSpawnScript [
   &PartitionParticles(Number Of Partitions=1)
   &InitializeParticle(
    Color=&MultiplyLinearColor(Linear Color=<rgba 0.7031 0.7031 0.7031 0.8>, Scale Factor=&MakeVector4_FromFloat(Float=1))
    Lifetime Max=1.5
    Lifetime Min=1.5
    Lifetime Mode=::NewEnumerator1
    Position Offset=<0 50 0>
    Sprite Rotation Mode=::NewEnumerator2
    Sprite Size Mode=::NewEnumerator1
    Uniform Sprite Size Max=150
    Uniform Sprite Size Min=100
    UsePositionOffset=true
    Write Material Random=true
   )
   &AddVelocity(
    Scale Added Velocity=&Multiply_VectorByFloat(Vector=&Multiply_VectorByFloat(Vector=<2.5 2.5 2.5>, Float=1), Float=1)
    Velocity=<0 0 -30>
   )
   &AddVelocityInCone(Cone Angle=50, Cone Axis=<0 0 -1>, Module.Use Velocity Falloff On Cone Axis=false, Velocity Strength=100) {module=AddVelocityInCone001}
   &CameraOffset()
   &ShapeLocation(MaskSpawnGroup=false, Override Seed=false, Sphere Radius=30, UseOwnerScale=true)
  ]
  particleUpdateScript [
   &SolveForcesAndVelocity(Clamp Velocity=false, Limit Acceleration=false)
   &ParticleState()
   &SpriteRotationRate(Rotation Rate=50)
   &ScaleSpriteSize(Uniform Curve Sprite Scale=DataInterface() {propertyValues {Curve {keys [
        {arriveTangent=10.8988, interpMode=RCIM_Cubic, leaveTangent=10.8988, tangentMode=RCTM_User, time=0, value=0}
        {arriveTangent=-1.0121, interpMode=RCIM_Cubic, leaveTangent=-1.0121, tangentMode=RCTM_User, time=0.2899, value=0.9265}
        <at 1 0.5>
       ]}}})
   &DynamicMaterialParameters(
    Index 0 Param 1=&FloatFromCurve(
     FloatCurve=DataInterface() {propertyValues {Curve {keys [
         {interpMode=RCIM_Cubic, tangentMode=RCTM_User, time=0, value=0}
         <at 0.2 0.55>
         {arriveTangent=-1.0223, interpMode=RCIM_Cubic, leaveTangent=-1.0223, tangentMode=RCTM_User, time=1, value=1}
        ]}}}
     Scale Curve=&RandomRangeFloat(Minimum=3, Maximum=3)
    )
    Module.Index0Param0Write=true
    Module.Index0Param1Write=true
    Module.Index0Param2Write=true
    Module.Index0Param3Write=true
   )
   &ScaleColor(
    Scale Alpha=&FloatFromCurve(
     FloatCurve=DataInterface() {propertyValues {Curve {keys [
         {arriveTangent=10.2356, interpMode=RCIM_Cubic, leaveTangent=10.2356, tangentMode=RCTM_User, time=0, value=0}
         <at 0.3 1>
         <at 0.7 0.4>
         <at 1 0>
        ]}}}
     Scale Curve=&Multiply_Float(A=0.5, B=1)
    )
    Scale RGB=&MakeVectorFromLinearColorRGB(Color=&ColorFromCurve(ColorCurve=DataInterface() {propertyValues {
        AlphaCurve {keys []}
        BlueCurve {keys [{arriveTangent=-331.6383, interpMode=RCIM_Cubic, time=0, value=0.0781}, <at 0.7598 0 RCIM_Cubic>]}
        GreenCurve {keys [{arriveTangent=-87.3695, interpMode=RCIM_Cubic, time=0, value=0.0781}, <at 0.7598 0 RCIM_Cubic>]}
        RedCurve {keys [<at 0 0.0781 RCIM_Cubic>,<at 0.7598 0 RCIM_Cubic>]}
       }}))
    ScaleA=true
    ScaleRGB=true
   )
   &SubUVAnimation(
    "Sprite Renderer "=DataInterface() {propertyValues {SpriteRenderer=&self:SmokeNoLocal_6.NiagaraSpriteRendererProperties_0}}
    UseEndFrame=true
    UseStartFrame=true
   )
  ]
  rendererClasses [&NiagaraSpriteRendererProperties]
  renderers [{rendererClass=&NiagaraSpriteRendererProperties}]
 }
 Smokedetail {
  emitterUpdateScript [
   &EmitterState(Inactive Response=::NewEnumerator1, Life Cycle Mode=::NewEnumerator1, Loop Delay=0.5, Loop Duration=1, UseLoopDelay=false)
   &SpawnBurst_Instantaneous(Use Spawn Probability=false, UseLoopCountLimit=false)
   &SpawnRate(
    Spawn Probability=&FN_FrontEndFloatModifier(Gameplay=1, Front End=0)
    SpawnRate=1.7
    Use Spawn Probability=true
   )
  ]
  particleSpawnScript [
   &PartitionParticles(Number Of Partitions=1)
   &InitializeParticle(
    Color=&MultiplyLinearColor(Linear Color=<rgba 0.7031 0.7031 0.7031 0.8>, Scale Factor=&MakeVector4_FromFloat(Float=1))
    Lifetime Max=1.6
    Lifetime Min=1.5
    Lifetime Mode=::NewEnumerator1
    Position Offset=<0 0 50>
    Sprite Rotation Mode=::NewEnumerator2
    Sprite Size Mode=::NewEnumerator1
    Uniform Sprite Size Max=150
    Uniform Sprite Size Min=150
    UsePositionOffset=true
    Write Material Random=true
   )
   &AddVelocity(
    Scale Added Velocity=&Multiply_VectorByFloat(Vector=&Multiply_VectorByFloat(Vector=<2.5 2.5 2.5>, Float=1), Float=1)
    Velocity=<0 0 -50>
   )
   &AddVelocityInCone(Cone Angle=60, Cone Axis=<0 0 -1>, Module.Use Velocity Falloff On Cone Axis=false, Velocity Strength=100) {module=AddVelocityInCone001}
   &CameraOffset()
   &ShapeLocation(MaskSpawnGroup=false, Override Seed=false, Sphere Radius=45, UseOwnerScale=true)
  ]
  particleUpdateScript [
   &Drag(Drag=0, Use Linear Drag=true, Use Rotational Drag=true)
   &SolveForcesAndVelocity(Clamp Velocity=false, Limit Acceleration=false)
   &ParticleState()
   &SpriteRotationRate(Rotation Rate=&RandomRangeFloat(Minimum=-50, Maximum=50))
   &ScaleSpriteSize(Uniform Curve Sprite Scale=DataInterface() {propertyValues {Curve {keys [
        {arriveTangent=15.2515, interpMode=RCIM_Cubic, leaveTangent=15.2515, tangentMode=RCTM_User, time=0, value=0}
        {arriveTangent=-0.0798, interpMode=RCIM_Cubic, leaveTangent=-0.0798, tangentMode=RCTM_User, time=0.3, value=1.125}
        <at 1 1>
       ]}}})
   &DynamicMaterialParameters(
    Index 0 Param 1=&FloatFromCurve(
     FloatCurve=DataInterface() {propertyValues {Curve {keys [
         {interpMode=RCIM_Cubic, tangentMode=RCTM_User, time=0, value=0}
         <at 0.2 0.55>
         {arriveTangent=-1.0223, interpMode=RCIM_Cubic, leaveTangent=-1.0223, tangentMode=RCTM_User, time=1, value=1}
        ]}}}
     Scale Curve=&RandomRangeFloat(Minimum=3, Maximum=3)
    )
    Module.Index0Param0Write=true
    Module.Index0Param1Write=true
    Module.Index0Param2Write=true
    Module.Index0Param3Write=true
   )
   &ScaleColor(
    Scale Alpha=&FloatFromCurve(
     FloatCurve=DataInterface() {propertyValues {Curve {keys [
         {arriveTangent=3.6424, interpMode=RCIM_Cubic, leaveTangent=3.6424, tangentMode=RCTM_User, time=0, value=0}
         <at 0.5 1>
         {arriveTangent=-2, interpMode=RCIM_Cubic, leaveTangent=-2, time=0.7096, value=0.5753}
         <at 1 0>
        ]}}}
     Scale Curve=&RandomRangeFloat(Minimum=0.5, Maximum=1)
    )
    Scale RGB=&MakeCustomVectorFromBool(
     Bool=$User.bIsFrontEndPreview
     True Vector=&MakeVectorFromLinearColorRGB(Color=&ColorFromCurve(ColorCurve=DataInterface() {propertyValues {
         AlphaCurve {keys []}
         BlueCurve {keys [{arriveTangent=-331.6383, interpMode=RCIM_Cubic, time=0, value=0.07}, <at 0.7598 0 RCIM_Cubic>]}
         GreenCurve {keys [{arriveTangent=-87.3695, interpMode=RCIM_Cubic, time=0, value=0.0546}, <at 0.7598 0 RCIM_Cubic>]}
         RedCurve {keys [<at 0 0.042 RCIM_Cubic>,<at 0.7598 0 RCIM_Cubic>]}
        }}))
     False Vector=&MakeVectorFromLinearColorRGB(Color=&FNTimeOfDayColor(
       Night Color=&ColorFromCurve(ColorCurve=DataInterface() {propertyValues {
          AlphaCurve {keys []}
          BlueCurve {keys [{arriveTangent=-331.6383, interpMode=RCIM_Cubic, time=0, value=0.1}, <at 0.7598 0 RCIM_Cubic>]}
          GreenCurve {keys [{arriveTangent=-87.3695, interpMode=RCIM_Cubic, time=0, value=0.055}, <at 0.7598 0 RCIM_Cubic>]}
          RedCurve {keys [<at 0 0.04 RCIM_Cubic>,<at 0.7598 0 RCIM_Cubic>]}
         }})
       Morning Color=&ColorFromCurve(ColorCurve=DataInterface() {propertyValues {
          AlphaCurve {keys []}
          BlueCurve {keys [{arriveTangent=-331.6383, interpMode=RCIM_Cubic, time=0, value=0.0781}, <at 0.7598 0 RCIM_Cubic>]}
          GreenCurve {keys [{arriveTangent=-87.3695, interpMode=RCIM_Cubic, time=0, value=0.0781}, <at 0.7598 0 RCIM_Cubic>]}
          RedCurve {keys [<at 0 0.0781 RCIM_Cubic>,<at 0.7598 0 RCIM_Cubic>]}
         }})
       Day Color=&ColorFromCurve(ColorCurve=DataInterface() {propertyValues {
          AlphaCurve {keys []}
          BlueCurve {keys [{arriveTangent=-331.6383, interpMode=RCIM_Cubic, time=0, value=0.0781}, <at 0.7598 0 RCIM_Cubic>]}
          GreenCurve {keys [{arriveTangent=-87.3695, interpMode=RCIM_Cubic, time=0, value=0.0781}, <at 0.7598 0 RCIM_Cubic>]}
          RedCurve {keys [<at 0 0.0781 RCIM_Cubic>,<at 0.7598 0 RCIM_Cubic>]}
         }})
       Evening Color=&ColorFromCurve(ColorCurve=DataInterface() {propertyValues {
          AlphaCurve {keys []}
          BlueCurve {keys [{arriveTangent=-331.6383, interpMode=RCIM_Cubic, time=0, value=0.0781}, <at 0.7598 0 RCIM_Cubic>]}
          GreenCurve {keys [{arriveTangent=-87.3695, interpMode=RCIM_Cubic, time=0, value=0.0781}, <at 0.7598 0 RCIM_Cubic>]}
          RedCurve {keys [<at 0 0.0781 RCIM_Cubic>,<at 0.7598 0 RCIM_Cubic>]}
         }})
      ))
    )
    ScaleA=true
    ScaleRGB=true
   )
   &SubUVAnimation(
    End Frame Range Override=30
    "Sprite Renderer "=DataInterface() {propertyValues {SpriteRenderer=&self:Smokedetail_8.NiagaraSpriteRendererProperties_0}}
    Start Frame Range Override=&RandomRangeInt(Minimum=0, Maximum=5)
    UseEndFrame=true
    UseStartFrame=true
   )
  ]
  rendererClasses [&NiagaraSpriteRendererProperties]
  renderers [{rendererClass=&NiagaraSpriteRendererProperties}]
 }
 Spiral_Trail {
  emitterUpdateScript [
   &EmitterState(
    Inactive Response=::NewEnumerator1
    Life Cycle Mode=::NewEnumerator1
    Loop Behavior=::NewEnumerator1
    Loop Delay=0.2
    Loop Duration Mode=::NewEnumerator1
    UseLoopDelay=true
   )
   &SpawnBurst_Instantaneous(Spawn Count=3, Use Spawn Probability=false, UseLoopCountLimit=false)
   &SpawnRate(SpawnRate=10, Use Spawn Probability=true)
   /SetVariables(Emitter.FEPositionFix:NiagaraFloat=&FN_FrontEndFloatModifier(Gameplay=1, Front End=1)) {isSetParameters=true}
   /SetVariables(Emitter.AngleFade:NiagaraFloat=&FN_FrontEndFloatModifier(Gameplay=0, Front End=2)) {isSetParameters=true}
  ]
  particleSpawnScript [
   &InitializeParticle(
    Color=&Multiply_LinearColorByFloat(
     Linear Color=<rgba 0.1192 0.128 0.13 3>
     Float=&FN_PickFloatDependOnPlatform(TargetPlatformActive=1, TargetPlatformNoActive=1.3)
    )
    Lifetime Mode=::NewEnumerator1
    Mesh Scale Mode=::NewEnumerator1
    Mesh Uniform Scale Max=1
    Mesh Uniform Scale Min=0.7
    Position Offset=&RandomRangeVector(Minimum=<0 0 10>, Maximum=<0 0 30>)
    UsePositionOffset=false
    Write Material Random=true
   )
   &CameraOffset(Camera Offset Amount=10)
   &InitialMeshOrientation(Mesh Orientation Mode=::NewEnumerator2, Rotation=<0 0.25 0>, Use Rotation Vector=true)
   &InitialMeshOrientation(
    Mesh Orientation Mode=::NewEnumerator2
    Rotation=&RandomRangeVector(Minimum=<0 0 0>, Maximum=<0 0 1>)
    Use Rotation Vector=true
   ) {module=InitialMeshOrientation001}
   &ShapeLocation(
    MaskSpawnGroup=false
    Offset Mode=::NewEnumerator0
    Override Seed=false
    Ring Radius=15
    Rotation Axis=<0 1 0>
    Rotation Mode=::NewEnumerator1
    Shape Origin=&VectorFromFloat(Value=0)
    Shape Primitive=::NewEnumerator4
    UseOwnerScale=true
   )
  ]
  particleUpdateScript [
   &SkeletalMeshLocation(
    Bone / Socket Index=5
    MaskSpawnGroup=false
    Mesh Sampling Type=::NewEnumerator5
    Sampled Position Offset=&AddVectorToPosition(
     Position=&AddVectorToPosition(
      Position=&RandomRangeVector(
       Minimum=<0 0 -30>
       Maximum=&FN_FrontEndVector3Modifier(Gameplay=<0 -10 35>, Front End=<0 -10 50>)
      )
      Vector=$Particles.Initial.ShapeLocation.ShapePosition
     )
     Vector=<0 0 -40>
    )
    Skeletal Mesh=$System.Skeletal Mesh
    Socket Sampling Mode=::NewEnumerator5
   )
   &Drag(Drag=10, Use Linear Drag=true, Use Rotational Drag=true)
   &SolveForcesAndVelocity(Clamp Velocity=false, Limit Acceleration=false)
   &ParticleState()
   &UpdateMeshOrientation(Rotation Rate=&RandomRangeFloat(Minimum=0, Maximum=1.5))
   &ScaleMeshSize(Scale Factor=&Multiply_VectorByFloat(
     Vector=&Multiply_Vector(A=<1.5 1 1>, B=<2 2 2>)
     Float=&Multiply_Float(A=1, B=1.3)
    ))
   &ScaleColor(
    Scale Alpha=&FloatFromCurve(
     FloatCurve=DataInterface() {propertyValues {Curve {keys [
         {interpMode=RCIM_Cubic, tangentMode=RCTM_User, time=0, value=0}
         {interpMode=RCIM_Cubic, tangentMode=RCTM_User, time=0.5, value=0.8}
         {tangentMode=RCTM_User, time=1, value=0}
        ]}}}
     Scale Curve=&Multiply_Float(A=0.8, B=&FN_FrontEndFloatModifier(Gameplay=1, Front End=1))
    )
    Scale RGB=&MakeCustomVectorFromBool(
     Bool=$User.bIsFrontEndPreview
     True Vector=&Multiply_VectorByFloat(
      Vector=&MakeVectorFromLinearColorRGB(Color=&ColorFromCurve(ColorCurve=DataInterface() {propertyValues {
          AlphaCurve {keys []}
          BlueCurve {keys [<at 0.4142 0 RCIM_Cubic>,<at 1 1 RCIM_Cubic>]}
          GreenCurve {keys [<at 0.4142 0 RCIM_Cubic>,<at 1 0.8187 RCIM_Cubic>]}
          RedCurve {keys [<at 0.4142 0 RCIM_Cubic>,<at 1 0.6771 RCIM_Cubic>]}
         }}))
      Float=&RandomRangeFloat(Minimum=0.5, Maximum=1)
     )
     False Vector=&Multiply_VectorByFloat(
      Vector=&MakeVectorFromLinearColorRGB(Color=&FNTimeOfDayColor(
        Night Color=&ColorFromCurve(ColorCurve=DataInterface() {propertyValues {
           AlphaCurve {keys []}
           BlueCurve {keys [<at 0.4142 0 RCIM_Cubic>,<at 1 1.5 RCIM_Cubic>]}
           GreenCurve {keys [<at 0.4142 0 RCIM_Cubic>,<at 1 0.7125 RCIM_Cubic>]}
           RedCurve {keys [<at 0.4142 0 RCIM_Cubic>,<at 1 0.45 RCIM_Cubic>]}
          }})
        Morning Color=&ColorFromCurve(ColorCurve=DataInterface() {propertyValues {
           AlphaCurve {keys []}
           BlueCurve {keys [<at 0.4142 0 RCIM_Cubic>,<at 1 1 RCIM_Cubic>]}
           GreenCurve {keys [<at 0.4142 0 RCIM_Cubic>,<at 1 1 RCIM_Cubic>]}
           RedCurve {keys [<at 0.4142 0 RCIM_Cubic>,<at 1 1 RCIM_Cubic>]}
          }})
        Day Color=&ColorFromCurve(ColorCurve=DataInterface() {propertyValues {
           AlphaCurve {keys []}
           BlueCurve {keys [<at 0.4142 0 RCIM_Cubic>,<at 1 1 RCIM_Cubic>]}
           GreenCurve {keys [<at 0.4142 0 RCIM_Cubic>,<at 1 1 RCIM_Cubic>]}
           RedCurve {keys [<at 0.4142 0 RCIM_Cubic>,<at 1 1 RCIM_Cubic>]}
          }})
        Evening Color=&ColorFromCurve(ColorCurve=DataInterface() {propertyValues {
           AlphaCurve {keys []}
           BlueCurve {keys [<at 0.4142 0 RCIM_Cubic>,<at 1 1 RCIM_Cubic>]}
           GreenCurve {keys [<at 0.4142 0 RCIM_Cubic>,<at 1 1 RCIM_Cubic>]}
           RedCurve {keys [<at 0.4142 0 RCIM_Cubic>,<at 1 1 RCIM_Cubic>]}
          }})
       ))
      Float=&RandomRangeFloat(Minimum=0.5, Maximum=1)
     )
    )
    ScaleA=true
    ScaleRGB=true
   )
   &DynamicMaterialParameters(
    Index 0 Param 3=1
    Index 0 Param 4=&FloatFromCurve(FloatCurve=DataInterface() {propertyValues {Curve {keys [<at 0 0>,<at 1 1>]}}})
    Index 2 Param 1=&MakeFloatFromVector(
     Vector=&AddVectorToPosition(
      Position=$Particles.Position
      Vector=&TransformVector(
       Vector=&FN_FrontEndVector3Modifier(Gameplay=$System.IGCenterposition, Front End=$System.FECenterposition)
       Destination Space=::Simulation
      )
     )
     Channel=::NewEnumerator0
    )
    Index 2 Param 2=&MakeFloatFromVector(
     Vector=&AddVectorToPosition(
      Position=$Particles.Position
      Vector=&TransformVector(
       Vector=&FN_FrontEndVector3Modifier(Gameplay=$System.IGCenterposition, Front End=$System.FECenterposition)
       Destination Space=::Simulation
      )
     )
     Channel=::NewEnumerator1
    )
    Index 2 Param 3=&MakeFloatFromVector(
     Vector=&AddVectorToPosition(
      Position=$Particles.Position
      Vector=&TransformVector(
       Vector=&FN_FrontEndVector3Modifier(Gameplay=$System.IGCenterposition, Front End=$System.FECenterposition)
       Destination Space=::Simulation
      )
     )
     Channel=::NewEnumerator2
    )
    Index 3 Param 1=150
    Module.Index0Param0Write=true
    Module.Index0Param1Write=true
    Module.Index0Param2Write=true
    Module.Index0Param3Write=true
    Module.Index2Param0Write=true
    Module.Index2Param1Write=true
    Module.Index2Param2Write=true
    Module.Index2Param3Write=true
    Module.Index3Param0Write=true
    Module.Index3Param1Write=true
    Module.Index3Param2Write=true
    Module.Index3Param3Write=true
    Write Parameter Index 2=true
    Write Parameter Index 3=true
   )
   &InterpolateOverTime(
    Rate Of Change:NiagaraFloat=&FN_FrontEndFloatModifier(
     Gameplay=&Multiply_Float(A=5, B=$System.InterpolateOverTime.InterpolateOverTime_Float.Moving Average)
     Front End=8
    )
    Target Vector Value:Vector3f=&AddVectorToPosition(
     Position=$Particles.Position
     Vector=&TransformVector(
      Vector=&FN_FrontEndVector3Modifier(Gameplay=$System.IGCenterposition, Front End=$System.FECenterposition)
      Destination Space=::Simulation
     )
    )
    Variable Type=::NewEnumerator2
   )
   /SetVariables(Particles.Interpolate Dynamic=&MakeFloat4FromVectorAndFloat(
     "Vector (RGB)"=$StackContext.InterpolateOverTime.InterpolateOverTime_V3_.Moving Average
     "Float (Alpha)"=&FN_FrontEndFloatModifier(Gameplay=8, Front End=2)
    )) {isSetParameters=true}
  ]
  rendererClasses [&NiagaraMeshRendererProperties]
  renderers [{rendererClass=&NiagaraMeshRendererProperties, rendererIndex=0}]
 }
 TrailShort {
  emitterUpdateScript [
   &EmitterState(
    Inactive Response=::NewEnumerator1
    Life Cycle Mode=::NewEnumerator1
    Loop Behavior=::NewEnumerator1
    Loop Delay=0.2
    Loop Duration Mode=::NewEnumerator1
    UseLoopDelay=true
   )
   &SpawnBurst_Instantaneous(Spawn Count=8, Use Spawn Probability=false, UseLoopCountLimit=false)
   /SetVariables(Emitter.FEPositionFix:NiagaraFloat=&FN_FrontEndFloatModifier(Gameplay=1, Front End=1)) {isSetParameters=true}
   /SetVariables(Emitter.AngleFade:NiagaraFloat=&FN_FrontEndFloatModifier(Gameplay=1, Front End=3)) {isSetParameters=true}
  ]
  particleSpawnScript [
   &PartitionParticles(Number Of Partitions=8)
   &InitializeParticle(
    Color Maximum=<rgba 0.015 0.015 0.015 3>
    Color Minimum=<rgba 0.01 0.01 0.01 2>
    Color Mode=::NewEnumerator2
    Mesh Scale=&SelectVectorFromArray(
     Direct Array Index=$Particles.PartitionParticles.Partition
     Vector Selection Array=DataInterface() {propertyValues {FloatData [<1 2.2 2>, <1 1.8 1.5>, <1 1.3 1>, <1 1.5 1>, <1 1.8 1>, <1 1.5 1>, <1 1.2 1>, <1 1.3 1>]}}
    )
    Mesh Scale Mode=::NewEnumerator4
    UsePositionOffset=true
    Write Material Random=true
   )
   &InitialMeshOrientation(Mesh Orientation Mode=::NewEnumerator2, Rotation=<-0.25 0 0>, Use Rotation Vector=true)
   &InitialMeshOrientation(
    Mesh Orientation Mode=::NewEnumerator2
    Rotation=&SelectVectorFromArray(
     Direct Array Index=$Particles.PartitionParticles.Partition
     Vector Selection Array=DataInterface() {propertyValues {FloatData [<0 -0.08 0>, <0 0.1 0>, <0 -0.08 0>, <0.05 0.07 0>, <0.03 -0.05 0>, <0 -0.05 0>, <0.03 0.05 0>, <-0.03 -0.03 0>]}}
    )
    Use Rotation Vector=true
   ) {module=InitialMeshOrientation001}
   &CameraOffset(Camera Offset Amount=20)
  ]
  particleUpdateScript [
   &ParticleState(Kill Particles When Lifetime Has Elapsed=false, Let Infinitely Lived Particles Die When Emitter Deactivates=true)
   &ScaleMeshSize(Scale Factor=&Multiply_VectorByFloat(Vector=&FN_FrontEndVector3Modifier(Gameplay=<1 0.5 1>, Front End=<1 0.8 1>), Float=0.7))
   &SkeletalMeshLocation(
    Bone / Socket Index=5
    Bone Sampling Mode=::NewEnumerator4
    MaskSpawnGroup=false
    Sampled Position Offset=&AddVectorToPosition(
     Position=&SelectVectorFromArray(
      Direct Array Index=$Particles.PartitionParticles.Partition
      Vector Selection Array=DataInterface() {propertyValues {FloatData [<5 0 70>, <-15 0 30>, <25 10 10>, <-15 -20 20>, <25 -20 -10>, <25 10 -30>, <-25 0 -50>, <-35 -15 -15>]}}
     )
     Vector=<0 0 -50>
    )
    Skeletal Mesh=$System.Skeletal Mesh
   )
   &DynamicMaterialParameters(
    Index 0 Param 1=&RandomRangeFloat(Minimum=0.2, Maximum=1)
    Index 0 Param 2=&RandomRangeFloat(Minimum=0.2, Maximum=1)
    Index 0 Param 3=&FN_FrontEndFloatModifier(Gameplay=0.5, Front End=0)
    Index 0 Param 4=20
    Index 2 Param 1=&MakeFloatFromVector(
     Vector=&AddVectorToPosition(
      Position=$Particles.Position
      Vector=&TransformVector(
       Vector=&FN_FrontEndVector3Modifier(Gameplay=$System.IGCenterposition, Front End=$System.FECenterposition)
       Destination Space=::Simulation
      )
     )
     Channel=::NewEnumerator0
    )
    Index 2 Param 2=&MakeFloatFromVector(
     Vector=&AddVectorToPosition(
      Position=$Particles.Position
      Vector=&TransformVector(
       Vector=&FN_FrontEndVector3Modifier(Gameplay=$System.IGCenterposition, Front End=$System.FECenterposition)
       Destination Space=::Simulation
      )
     )
     Channel=::NewEnumerator1
    )
    Index 2 Param 3=&MakeFloatFromVector(
     Vector=&AddVectorToPosition(
      Position=$Particles.Position
      Vector=&TransformVector(
       Vector=&FN_FrontEndVector3Modifier(Gameplay=$System.IGCenterposition, Front End=$System.FECenterposition)
       Destination Space=::Simulation
      )
     )
     Channel=::NewEnumerator2
    )
    Index 3 Param 1=150
    Index 3 Param 4=&FloatFromCurve(FloatCurve=DataInterface() {propertyValues {Curve {keys [<at 0 0>,<at 1 1>]}}})
    Module.Index0Param0Write=true
    Module.Index0Param1Write=true
    Module.Index0Param2Write=true
    Module.Index0Param3Write=true
    Module.Index2Param0Write=true
    Module.Index2Param1Write=true
    Module.Index2Param2Write=true
    Module.Index2Param3Write=true
    Module.Index3Param0Write=true
    Module.Index3Param1Write=true
    Module.Index3Param2Write=true
    Module.Index3Param3Write=true
    Write Parameter Index 2=true
    Write Parameter Index 3=true
   )
   &InterpolateOverTime(
    Rate Of Change:NiagaraFloat=&FN_FrontEndFloatModifier(
     Gameplay=&Multiply_Float(A=10, B=$System.InterpolateOverTime.InterpolateOverTime_Float.Moving Average)
     Front End=10
    )
    Target Vector Value:Vector3f=&AddVectorToPosition(
     Position=$Particles.Position
     Vector=&TransformVector(
      Vector=&FN_FrontEndVector3Modifier(Gameplay=$System.IGCenterposition, Front End=$System.FECenterposition)
      Destination Space=::Simulation
     )
    )
    Variable Type=::NewEnumerator2
   )
   /SetVariables(Particles.Interpolate Dynamic=&MakeFloat4FromVectorAndFloat(
     "Vector (RGB)"=$StackContext.InterpolateOverTime.InterpolateOverTime_V3_.Moving Average
     "Float (Alpha)"=&FN_FrontEndFloatModifier(Gameplay=8, Front End=2)
    )) {isSetParameters=true, module=SetVariables_#4}
   /SetVariables(Particles.Color002=&MakeCustomLinearColorFromBool(
     Bool=$User.bIsFrontEndPreview
     True Color=<rgba 0.01 0.01 0.01 0.6>
     False Color=&FNTimeOfDayColor(
      Night Color=<rgba 0.0161 0.0242 0.05 0.6>
      Morning Color=<rgba 0.03 0.03 0.03 0.6>
      Day Color=<rgba 0.03 0.03 0.03 0.6>
      Evening Color=<rgba 0.03 0.03 0.03 0.6>
     )
    )) {isSetParameters=true, module=SetVariables_#5}
   &ScaleColor(
    Scale Alpha=&FloatFromCurve(
     FloatCurve=DataInterface() {propertyValues {Curve {keys [<at 0 0>,<at 0.1 1>]}}}
     Scale Curve=&SelectFloatFromArray(
      Direct Array Index=$Particles.PartitionParticles.Partition
      Float Selection Array=DataInterface() {propertyValues {FloatData [1,1,1,1,1,1,1,1]}}
     )
    )
    Scale RGB=&FN_FrontEndVector3Modifier(
     Gameplay=&MakeVectorFromLinearColorRGB(Color=&FNTimeOfDayColor(Night Color=<rgba 0.3 0.475 1 1>, Morning Color=<rgba 1 1 1 1>, Day Color=<rgba 1 1 1 1>, Evening Color=<rgba 1 1 1 1>))
     Front End=<0 0 0>
    )
    ScaleA=true
    ScaleRGB=true
   )
  ]
  rendererClasses [&NiagaraMeshRendererProperties]
  renderers [
   {rendererClass=&NiagaraMeshRendererProperties, rendererIndex=0}
   {rendererClass=&NiagaraMeshRendererProperties, rendererIndex=1}
  ]
 }
 Trail_Long {
  emitterUpdateScript [
   &EmitterState(
    Inactive Response=::NewEnumerator1
    Life Cycle Mode=::NewEnumerator1
    Loop Behavior=::NewEnumerator1
    Loop Delay=0.3
    Loop Duration Mode=::NewEnumerator1
    UseLoopDelay=true
   )
   &SpawnBurst_Instantaneous(Spawn Count=4, Use Spawn Probability=false, UseLoopCountLimit=false)
   /SetVariables(Emitter.FEPositionFix:NiagaraFloat=&FN_FrontEndFloatModifier(Gameplay=1, Front End=1)) {isSetParameters=true}
  ]
  particleSpawnScript [
   &PartitionParticles(Number Of Partitions=6)
   &InitializeParticle(
    Color=<rgba 0.5156 0.5156 0.5156 3>
    Mesh Scale=&SelectVectorFromArray(
     Direct Array Index=$Particles.PartitionParticles.Partition
     Vector Selection Array=DataInterface() {propertyValues {FloatData [<0.7 0.7 0.8>, <0.7 0.7 0.6>, <0.7 0.7 0.9>, <0.7 0.7 0.8>]}}
    )
    Mesh Scale Mode=::NewEnumerator4
    UsePositionOffset=false
    Write Material Random=true
   )
   &DynamicMaterialParameters(
    Index 0 Param 1=&RandomRangeFloat(Minimum=0, Maximum=1)
    Index 0 Param 2=&RandomRangeFloat(Minimum=0, Maximum=10)
    Index 0 Param 3=30
    Index 0 Param 4=1
    Module.Param0WriteEnabled=true
    Module.Param1WriteEnabled=true
    Module.Param2WriteEnabled=false
    Module.Param3WriteEnabled=true
   )
   &InitialMeshOrientation(Mesh Orientation Mode=::NewEnumerator2, Rotation=<0.5 0 0>, Use Rotation Vector=true)
   &InitialMeshOrientation(
    Mesh Orientation Mode=::NewEnumerator2
    Rotation=&SelectVectorFromArray(
     Direct Array Index=$Particles.PartitionParticles.Partition
     Vector Selection Array=DataInterface() {propertyValues {FloatData [<-0.05 0.15 0.1>, <-0.05 -0.15 -0.1>, <-0.06 0.1 0>, <-0.05 -0.1 0>]}}
    )
    Use Rotation Vector=true
   ) {module=InitialMeshOrientation001}
  ]
  particleUpdateScript [
   &ParticleState(Kill Particles When Lifetime Has Elapsed=false, Let Infinitely Lived Particles Die When Emitter Deactivates=true)
   &ScaleMeshSize(Scale Factor=&Multiply_VectorByFloat(Vector=<1 1 1>, Float=1.1))
   &SkeletalMeshLocation(
    Bone / Socket Index=5
    Bone Sampling Mode=::NewEnumerator4
    MaskSpawnGroup=false
    Sampled Position Offset=&AddVectorToPosition(
     Position=&SelectVectorFromArray(
      Direct Array Index=$Particles.PartitionParticles.Partition
      Vector Selection Array=DataInterface() {propertyValues {FloatData [<0 20 15>, <3 10 5>, <-20 10 -15>, <20 5 -15>]}}
     )
     Vector=<0 -20 -10>
    )
    Skeletal Mesh=$System.Skeletal Mesh
   )
   &InterpolateOverTime(
    Rate Of Change:NiagaraFloat=&FN_FrontEndFloatModifier(
     Gameplay=&Multiply_Float(A=8, B=$System.InterpolateOverTime.InterpolateOverTime_Float.Moving Average)
     Front End=10
    )
    Target Vector Value:Vector3f=&AddVectorToPosition(
     Position=$Particles.Position
     Vector=&TransformVector(
      Vector=&FN_FrontEndVector3Modifier(Gameplay=$System.IGCenterposition, Front End=$System.FECenterposition)
      Destination Space=::Simulation
     )
    )
    Variable Type=::NewEnumerator2
   )
   /SetVariables(Particles.InterpolateDynamic:Vector4f=&MakeFloat4FromVectorAndFloat(
     "Vector (RGB)"=$StackContext.InterpolateOverTime.InterpolateOverTime_V3_.Moving Average
     "Float (Alpha)"=&FN_FrontEndFloatModifier(Gameplay=3, Front End=2)
    )) {isSetParameters=true}
   &DynamicMaterialParameters(
    Index 0 Param 1=&RandomRangeFloat(Minimum=0.5, Maximum=1)
    Index 0 Param 2=&RandomRangeFloat(Minimum=0.5, Maximum=1)
    Index 0 Param 3=&Multiply_Float(A=15, B=&FN_FrontEndFloatModifier(Gameplay=1, Front End=0.3))
    Index 0 Param 4=1
    Index 2 Param 1=&MakeFloatFromVector(
     Vector=&AddVectorToPosition(
      Position=$Particles.Position
      Vector=&TransformVector(
       Vector=&FN_FrontEndVector3Modifier(Gameplay=$System.IGCenterposition, Front End=$System.FECenterposition)
       Destination Space=::Simulation
      )
     )
     Channel=::NewEnumerator0
    )
    Index 2 Param 2=&MakeFloatFromVector(
     Vector=&AddVectorToPosition(
      Position=$Particles.Position
      Vector=&TransformVector(
       Vector=&FN_FrontEndVector3Modifier(Gameplay=$System.IGCenterposition, Front End=$System.FECenterposition)
       Destination Space=::Simulation
      )
     )
     Channel=::NewEnumerator1
    )
    Index 2 Param 3=&MakeFloatFromVector(
     Vector=&AddVectorToPosition(
      Position=$Particles.Position
      Vector=&TransformVector(
       Vector=&FN_FrontEndVector3Modifier(Gameplay=$System.IGCenterposition, Front End=$System.FECenterposition)
       Destination Space=::Simulation
      )
     )
     Channel=::NewEnumerator2
    )
    Index 3 Param 1=150
    Index 3 Param 4=&FloatFromCurve(FloatCurve=DataInterface() {propertyValues {Curve {keys [<at 0 0>,<at 1 1>]}}})
    Module.Param0WriteEnabled=true
    Module.Param0WriteEnabled2=true
    Module.Param0WriteEnabled3=true
    Module.Param1WriteEnabled=true
    Module.Param1WriteEnabled2=true
    Module.Param1WriteEnabled3=true
    Module.Param2WriteEnabled=true
    Module.Param2WriteEnabled2=true
    Module.Param2WriteEnabled3=true
    Module.Param3WriteEnabled=true
    Module.Param3WriteEnabled2=false
    Module.Param3WriteEnabled3=true
    Write Parameter Index 2=true
    Write Parameter Index 3=true
   ) {module=DynamicMaterialParameters001}
   &SolveForcesAndVelocity(Clamp Velocity=false, Limit Acceleration=false)
   &ScaleColor(
    Scale RGB=&MakeVectorFromLinearColorRGB(Color=&FNTimeOfDayColor(Night Color=<rgba 0.6 0.95 2 1>, Morning Color=<rgba 1 1 1 1>, Day Color=<rgba 1 1 1 1>, Evening Color=<rgba 1 1 1 1>))
    ScaleA=true
    ScaleRGB=true
   )
  ]
  rendererClasses [&NiagaraMeshRendererProperties]
  renderers [{rendererClass=&NiagaraMeshRendererProperties, rendererIndex=0}]
 }
}
summary {
 systemName=NS_Emote_MistMylarLoop
 emitters [
  {emitterName=BodyCore, rendererClasses [&NiagaraMeshRendererProperties]}
  {emitterName=Spiral_Trail, rendererClasses [&NiagaraMeshRendererProperties]}
  {emitterName=Cloud, rendererClasses [&NiagaraMeshRendererProperties]}
  {emitterName=TrailShort, rendererClasses [&NiagaraMeshRendererProperties]}
  {emitterName=SmokeLocal, rendererClasses [&NiagaraSpriteRendererProperties]}
  {emitterName=MaskTrails, rendererClasses [&NiagaraMeshRendererProperties]}
  {emitterName=Trail_Long, rendererClasses [&NiagaraMeshRendererProperties]}
  {emitterName=Smokedetail, rendererClasses [&NiagaraSpriteRendererProperties]}
  {bEnabled=false, emitterName=SmokeNoLocal, rendererClasses [&NiagaraSpriteRendererProperties]}
  {emitterName=RibbonTail, rendererClasses [&NiagaraRibbonRendererProperties]}
 ]
 userVariables [
  User.bIsFrontEnd {type=NiagaraBool}
  User.bIsFrontEndPreview {type=NiagaraBool}
 ]
}
