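# Niagara system for the deploy phase of the MistMylar emote: about a second of smoke and trails on the character, then a flash and downward smoke burst.
# Sprite smoke layers and partitioned bone-sampled ribbons build the plume, while hidden-style Source particles track a bone and feed BodyTrail ribbons via particle reads.
# Smoke tints follow time of day, with a front-end preview override on two layers.
# Cloud009 is disabled.
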NiagaraSystem /BRCosmetics/Effects/Fort_Effects/Emotes/MistMylar/NS_Emote_MistMylar_deploy
dependencies {
 usedDataInterfaces [
  &NiagaraDataInterfaceCurve
  &NiagaraDataInterfaceColorCurve
  &NiagaraDataInterfaceSpriteRendererInfo
  &NiagaraDataInterfaceArrayFloat
  &NiagaraDataInterfaceParticleRead
  &NiagaraDataInterfaceSkeletalMesh
 ]
 usedDynamicInputs []
 usedModules [
  &self:SystemSpawnScript.SystemScriptSource.SystemScriptGraph.NiagaraNodeAssignment_0.SetVariables_#1
  &SystemState
 ]
}
emitters {
 BodyTrail {
  emitterUpdateScript [
   &EmitterState(Life Cycle Mode=::NewEnumerator1, Loop Behavior=::NewEnumerator1, Loop Delay=0.3, Loop Duration=1.2, UseLoopDelay=true)
   &SpawnParticlesFromOtherEmitter(
    Attribute Reader=DataInterface() {propertyValues {EmitterBinding {bindingMode=Other, emitterName=Source}}}
    Spawn Rate=60
    Use Maximum Age=false
    Use Minimum Age=false
   )
  ]
  particleSpawnScript [
   &PartitionParticles()
   &InitializeParticle(
    Color=<rgba 0.6875 0.7385 0.75 3>
    Lifetime=0.6
    Ribbon Width=50
    Ribbon Width Mode=::NewEnumerator1
    UsePositionOffset=false
    Write Material Random=true
   )
   &SampleParticlesFromOtherEmitter(Particle ID Sampling=::NewEnumerator1)
  ]
  particleUpdateScript [
   &SolveForcesAndVelocity(Clamp Velocity=false, Limit Acceleration=false)
   &ParticleState()
   &ScaleColor(
    Scale Alpha=&FloatFromCurve(
     FloatCurve=DataInterface() {propertyValues {Curve {keys [
         {interpMode=RCIM_Cubic, tangentMode=RCTM_User, time=0.1, value=0}
         {arriveTangent=2, interpMode=RCIM_Cubic, leaveTangent=2, time=0.3, value=0.9}
         {interpMode=RCIM_Cubic, tangentMode=RCTM_User, time=0.5, value=0.8}
         {tangentMode=RCTM_User, time=1, value=0}
        ]}}}
     Scale Curve=&Multiply_Float(
      A=0.49
      B=&FloatFromCurve(FloatCurve=DataInterface() {propertyValues {Curve {keys [<at 0 1>,<at 1 1>,<at 1.3 0>]}}}, CurveIndex=$Emitter.Age)
     )
    )
    Scale RGB=&Multiply_VectorByFloat(
     Vector=&MakeVectorFromLinearColorRGB(Color=&ColorFromCurve(ColorCurve=DataInterface() {propertyValues {
         AlphaCurve {keys []}
         BlueCurve {keys [{arriveTangent=-331.6383, interpMode=RCIM_Cubic, time=0, value=0.08}, <at 0.2055 0 RCIM_Cubic>]}
         GreenCurve {keys [{arriveTangent=-87.3695, interpMode=RCIM_Cubic, time=0, value=0.08}, <at 0.2055 0 RCIM_Cubic>]}
         RedCurve {keys [<at 0 0.08 RCIM_Cubic>,<at 0.2055 0 RCIM_Cubic>]}
        }}))
     Float=&RandomRangeFloat(Minimum=1, Maximum=1)
    )
    ScaleA=true
    ScaleRGB=true
   ) {module=ScaleColor001}
   /SetVariables(Particles.Color002=&MakeLinearColorFromVectorAndFloat(
     "Vector (RGB)"=&MakeVectorFromLinearColorRGB(Color=&ColorFromCurve(ColorCurve=DataInterface() {propertyValues {
         AlphaCurve {keys []}
         BlueCurve {keys [<at 0 0.08>,<at 0.2068 0 RCIM_Cubic>]}
         GreenCurve {keys [<at 0 0.08>,<at 0.2068 0 RCIM_Cubic>]}
         RedCurve {keys [<at 0 0.08>,<at 0.2068 0 RCIM_Cubic>]}
        }}))
     "Float (Alpha)"=&Multiply_Float(
      A=1
      B=&FloatFromCurve(
       FloatCurve=DataInterface() {propertyValues {Curve {keys [
           {interpMode=RCIM_Cubic, tangentMode=RCTM_User, time=0.1, value=0}
           {arriveTangent=2, interpMode=RCIM_Cubic, leaveTangent=2, time=0.3, value=0.95}
           {interpMode=RCIM_Cubic, tangentMode=RCTM_User, time=0.5, value=0.8}
           {tangentMode=RCTM_User, time=0.8, value=0}
          ]}}}
       Scale Curve=&Multiply_Float(
        A=&FloatFromCurve(FloatCurve=DataInterface() {propertyValues {Curve {keys [<at 0 1>,<at 1 1>,<at 1.3 0>]}}}, CurveIndex=$Emitter.Age)
        B=0.7
       )
      )
     )
    )) {isSetParameters=true}
   /SetVariables(Particles.Width02=25) {isSetParameters=true}
  ]
  rendererClasses [&NiagaraRibbonRendererProperties]
  renderers [
   {rendererClass=&NiagaraRibbonRendererProperties}
   {rendererClass=&NiagaraRibbonRendererProperties, rendererIndex=1}
  ]
 }
 Cloud009 {
  bEnabled=false
  emitterUpdateScript [
   &EmitterState(Life Cycle Mode=::NewEnumerator1, Loop Behavior=::NewEnumerator1, Loop Delay=1.2, Loop Duration=1, UseLoopDelay=true)
   &SpawnBurst_Instantaneous(Spawn Count=5, Use Spawn Probability=false, UseLoopCountLimit=false)
  ]
  particleSpawnScript [
   &InitializeParticle(
    Color=&MultiplyLinearColor(Linear Color=<rgba 0.7031 0.7031 0.7031 0.8>, Scale Factor=&MakeVector4_FromFloat(Float=1))
    Lifetime Max=1
    Lifetime Min=0.8
    Lifetime Mode=::NewEnumerator1
    Sprite Rotation Mode=::NewEnumerator2
    Sprite Size Mode=::NewEnumerator1
    Uniform Sprite Size Max=120
    Uniform Sprite Size Min=100
    UsePositionOffset=true
    Write Material Random=true
   )
   &AddVelocityInCone(
    Cone Angle=100
    Cone Axis=<0 0 -1>
    Module.Use Velocity Falloff On Cone Axis=false
    Velocity Strength=&RandomRangeFloat(Minimum=100, Maximum=500)
   ) {module=AddVelocityInCone001}
   &CameraOffset(Camera Offset Amount=-20)
   &ShapeLocation(MaskSpawnGroup=false, Override Seed=false, Sphere Radius=30, UseOwnerScale=true)
  ]
  particleUpdateScript [
   &Drag(Drag=&RandomRangeFloat(Minimum=1, Maximum=1), Use Linear Drag=true, Use Rotational Drag=true)
   &InheritSourceMovement(Applied Position Scale=<0.5 0.5 0.5>, ApplyForce=false, ApplyPosition=true, ApplyPositionFalloff=false)
   &SolveForcesAndVelocity(Clamp Velocity=false, Limit Acceleration=false)
   &ParticleState()
   &SpriteRotationRate(Rotation Rate=50)
   &ScaleSpriteSize(Uniform Curve Sprite Scale=DataInterface() {propertyValues {Curve {keys [{interpMode=RCIM_Cubic, tangentMode=RCTM_User, time=0, value=1.2}, {interpMode=RCIM_Cubic, tangentMode=RCTM_User, time=1, value=1.3}]}}})
   &DynamicMaterialParameters(
    Index 0 Param 1=&RandomRangeFloat(Minimum=0, Maximum=1)
    Module.Index0Param0Write=true
    Module.Index0Param1Write=true
    Module.Index0Param2Write=true
    Module.Index0Param3Write=true
   )
   &ScaleColor(
    Scale Alpha=&Multiply_Float(
     A=&FloatFromCurve(
      FloatCurve=DataInterface() {propertyValues {Curve {keys [
          {arriveTangent=4.9971, interpMode=RCIM_Cubic, leaveTangent=4.9971, tangentMode=RCTM_User, time=0, value=0}
          <at 0.3 1>
          <at 0.7 0.25>
          <at 1 0>
         ]}}}
      Scale Curve=&RandomRangeFloat(Minimum=1, Maximum=2)
     )
     B=1
    )
    Scale RGB=&MakeVectorFromLinearColorRGB(Color=&ColorFromCurve(ColorCurve=DataInterface() {propertyValues {
        AlphaCurve {keys []}
        BlueCurve {keys [{arriveTangent=-331.6383, interpMode=RCIM_Cubic, time=0, value=0.2}, <at 0.6685 0 RCIM_Cubic>]}
        GreenCurve {keys [{arriveTangent=-87.3695, interpMode=RCIM_Cubic, time=0, value=0.2}, <at 0.6685 0 RCIM_Cubic>]}
        RedCurve {keys [<at 0 0.2 RCIM_Cubic>,<at 0.6685 0 RCIM_Cubic>]}
       }}))
    ScaleA=true
    ScaleRGB=true
   )
   &SubUVAnimation(
    End Frame Range Override=30
    "Sprite Renderer "=DataInterface() {propertyValues {SpriteRenderer=&self:Cloud009_9.NiagaraSpriteRendererProperties_1}}
    Start Frame Range Override=&RandomRangeInt(Minimum=0, Maximum=5)
    UseEndFrame=true
    UseStartFrame=true
   ) {module=SubUVAnimation001}
  ]
  rendererClasses [&NiagaraSpriteRendererProperties]
  renderers [{rendererClass=&NiagaraSpriteRendererProperties}]
 }
 Flash {
  emitterUpdateScript [
   &EmitterState(Life Cycle Mode=::NewEnumerator1, Loop Behavior=::NewEnumerator1, Loop Delay=1.2, Loop Duration=0.5, UseLoopDelay=true)
   &SpawnBurst_Instantaneous(Use Spawn Probability=false, UseLoopCountLimit=false)
  ]
  particleSpawnScript [
   &InitializeParticle(
    Color Mode=::NewEnumerator0
    Lifetime=0.2
    Sprite Size Mode=::NewEnumerator3
    Uniform Sprite Size=220
    UsePositionOffset=false
    Write Material Random=true
   )
   &CameraOffset(Camera Offset Amount=20)
  ]
  particleUpdateScript [
   &SkeletalMeshLocation(Bone / Socket Index=4, Bone Sampling Mode=::NewEnumerator4, MaskSpawnGroup=false, Skeletal Mesh=$System.Skeletal Mesh)
   &ParticleState()
   &ScaleColor(
    Scale Alpha=&FloatFromCurve(FloatCurve=DataInterface() {propertyValues {Curve {keys [<at 0 0 RCIM_Cubic>, {interpMode=RCIM_Cubic, leaveTangent=-3.5996, tangentMode=RCTM_Break, time=0.2, value=1.5}, <at 1 0>]}}})
    ScaleA=true
    ScaleRGB=true
   )
  ]
  rendererClasses [&NiagaraSpriteRendererProperties]
  renderers [{rendererClass=&NiagaraSpriteRendererProperties}]
 }
 SmokeBlur {
  emitterUpdateScript [
   &EmitterState(Life Cycle Mode=::NewEnumerator1, Loop Behavior=::NewEnumerator1, Loop Delay=0.3, Loop Duration=1.2, UseLoopDelay=true)
   &SpawnPerUnit(
    Module.Use Max Movement Threshold=true
    Module.Use Movement Tolerance=false
    Movement Tolerance=0
    Spawn Spacing=1
    Use Spawn Probability=false
   )
  ]
  particleSpawnScript [
   &PartitionParticles(Number Of Partitions=5)
   &InitializeParticle(
    Color=<rgba 0 0 0 1>
    Ribbon Width=30
    Ribbon Width Mode=::NewEnumerator1
    Sprite Size Mode=::NewEnumerator3
    Uniform Sprite Size=20
    UsePositionOffset=false
    Write Material Random=true
   )
   &SkeletalMeshLocation(
    Bone / Socket Index=$Particles.PartitionParticles.Partition
    Bone Sampling Mode=::NewEnumerator4
    MaskSpawnGroup=false
    Sampled Position Offset=<0 -10 0>
    Skeletal Mesh=$System.Skeletal Mesh
   )
   &DynamicMaterialParameters(
    Index 0 Param 1=&SelectFloatFromArray(
     Direct Array Index=$Particles.PartitionParticles.Partition
     Float Selection Array=DataInterface() {propertyValues {FloatData [1,1,0.5,0.2,0]}}
    )
    Index 0 Param 2=1
    Index 0 Param 4=20
    Index 3 Param 1=150
    Module.Index0Param0Write=true
    Module.Index0Param1Write=true
    Module.Index0Param2Write=true
    Module.Index0Param3Write=true
    Module.Index3Param0Write=true
    Module.Index3Param1Write=true
    Module.Index3Param2Write=true
    Module.Index3Param3Write=true
    Write Parameter Index 3=true
   )
   /SetVariables(Particles.DY2:Vector4f=&MakeVector4(
     X=-0.7
     Y=0
     Z=&SelectFloatFromArray(
      Direct Array Index=$Particles.PartitionParticles.Partition
      Float Selection Array=DataInterface() {propertyValues {FloatData [0.2,0.4,0.5,0.2,0]}}
     )
     W=1
    )) {isSetParameters=true}
   &AddVelocity(Velocity=<0 10 0>)
  ]
  particleUpdateScript [
   /SetVariables(Particles.Color002=&MakeLinearColorFromVectorAndFloat(
     "Vector (RGB)"=&MakeVectorFromLinearColorRGB(Color=&FNTimeOfDayColor(
       Night Color=<rgba 0.009 0.014 0.03 0.6>
       Morning Color=<rgba 0.03 0.03 0.03 0.6>
       Day Color=<rgba 0.03 0.03 0.03 0.6>
       Evening Color=<rgba 0.03 0.03 0.03 0.6>
      ))
     "Float (Alpha)"=&Multiply_Float(
      A=0.7
      B=&FloatFromCurve(FloatCurve=DataInterface() {propertyValues {Curve {keys [<at 0 1>,<at 1 1>,<at 1.5 0>]}}}, CurveIndex=$Emitter.Age, Scale Curve=1.2)
     )
    )) {isSetParameters=true}
   &SolveForcesAndVelocity(Clamp Velocity=false, Limit Acceleration=false)
   &ParticleState()
   &ScaleRibbonWidth()
   &ScaleColor(
    Scale Alpha=&FloatFromCurve(FloatCurve=DataInterface() {propertyValues {Curve {keys [<at 0 1>,<at 1 1>,<at 1.5 0>]}}}, CurveIndex=$Emitter.Age, Scale Curve=0.8)
    ScaleA=true
    ScaleRGB=true
   )
  ]
  rendererClasses [&NiagaraRibbonRendererProperties]
  renderers [
   {rendererClass=&NiagaraRibbonRendererProperties}
   {rendererClass=&NiagaraRibbonRendererProperties, rendererIndex=1}
  ]
 }
 SmokeBody02 {
  emitterUpdateScript [
   &EmitterState(Life Cycle Mode=::NewEnumerator1, Loop Behavior=::NewEnumerator1, Loop Delay=0.2, Loop Duration=1, UseLoopDelay=true)
   &SpawnPerUnit(Module.Use Max Movement Threshold=true, Module.Use Movement Tolerance=false, Spawn Spacing=30, Use Spawn Probability=false)
   &SpawnRate(SpawnRate=20, Use Spawn Probability=true)
  ]
  particleSpawnScript [
   &PartitionParticles(Number Of Partitions=1)
   &InitializeParticle(
    Color=&MultiplyLinearColor(Linear Color=<rgba 0.7031 0.7031 0.7031 0.8>, Scale Factor=&MakeVector4_FromFloat(Float=1))
    Lifetime Max=1.5
    Lifetime Min=0.7
    Lifetime Mode=::NewEnumerator1
    Sprite Rotation Mode=::NewEnumerator2
    Sprite Size Mode=::NewEnumerator1
    Uniform Sprite Size Max=30
    Uniform Sprite Size Min=50
    UsePositionOffset=true
    Write Material Random=true
   )
   &CameraOffset(Camera Offset Amount=-20)
   &ShapeLocation(MaskSpawnGroup=false, Override Seed=false, Sphere Radius=30, UseOwnerScale=true)
  ]
  particleUpdateScript [
   &SolveForcesAndVelocity(Clamp Velocity=false, Limit Acceleration=false)
   &ParticleState()
   &SpriteRotationRate(Rotation Rate=50)
   &ScaleSpriteSize(Uniform Curve Sprite Scale=DataInterface() {propertyValues {Curve {keys [
        {arriveTangent=10.8988, interpMode=RCIM_Cubic, leaveTangent=10.8988, tangentMode=RCTM_User, time=0, value=0}
        {arriveTangent=-1.0121, interpMode=RCIM_Cubic, leaveTangent=-1.0121, tangentMode=RCTM_User, time=0.2831, value=1.5381}
        {arriveTangent=-1.169, interpMode=RCIM_Cubic, leaveTangent=-1.169, time=0.5, value=0.875}
        <at 1 0.7>
       ]}}})
   &DynamicMaterialParameters(
    Index 0 Param 1=&RandomRangeFloat(Minimum=0, Maximum=1)
    Module.Index0Param0Write=true
    Module.Index0Param1Write=true
    Module.Index0Param2Write=true
    Module.Index0Param3Write=true
   )
   &ScaleColor(
    Scale Alpha=&Multiply_Float(
     A=&FloatFromCurve(
      FloatCurve=DataInterface() {propertyValues {Curve {keys [
          {arriveTangent=10.2356, interpMode=RCIM_Cubic, leaveTangent=10.2356, tangentMode=RCTM_User, time=0, value=0}
          <at 0.3 1>
          <at 0.7 0.4>
          <at 1 0>
         ]}}}
      Scale Curve=&RandomRangeFloat(Minimum=1, Maximum=2)
     )
     B=&FloatFromCurve(FloatCurve=DataInterface() {propertyValues {Curve {keys [<at 0 1>,<at 1.1 1>,<at 1.6 0>]}}}, CurveIndex=$Emitter.Age)
    )
    Scale RGB=&MakeVectorFromLinearColorRGB(Color=&FNTimeOfDayColor(
      Night Color=&ColorFromCurve(ColorCurve=DataInterface() {propertyValues {
         AlphaCurve {keys []}
         BlueCurve {keys [{arriveTangent=-331.6383, interpMode=RCIM_Cubic, time=0, value=0.1}, <at 0.209 0 RCIM_Cubic>]}
         GreenCurve {keys [{arriveTangent=-87.3695, interpMode=RCIM_Cubic, time=0, value=0.055}, <at 0.209 0 RCIM_Cubic>]}
         RedCurve {keys [<at 0 0.04 RCIM_Cubic>,<at 0.209 0 RCIM_Cubic>]}
        }})
      Morning Color=&ColorFromCurve(ColorCurve=DataInterface() {propertyValues {
         AlphaCurve {keys []}
         BlueCurve {keys [{arriveTangent=-331.6383, interpMode=RCIM_Cubic, time=0, value=0.1}, <at 0.209 0 RCIM_Cubic>]}
         GreenCurve {keys [{arriveTangent=-87.3695, interpMode=RCIM_Cubic, time=0, value=0.1}, <at 0.209 0 RCIM_Cubic>]}
         RedCurve {keys [<at 0 0.1 RCIM_Cubic>,<at 0.209 0 RCIM_Cubic>]}
        }})
      Day Color=&ColorFromCurve(ColorCurve=DataInterface() {propertyValues {
         AlphaCurve {keys []}
         BlueCurve {keys [{arriveTangent=-331.6383, interpMode=RCIM_Cubic, time=0, value=0.1}, <at 0.209 0 RCIM_Cubic>]}
         GreenCurve {keys [{arriveTangent=-87.3695, interpMode=RCIM_Cubic, time=0, value=0.1}, <at 0.209 0 RCIM_Cubic>]}
         RedCurve {keys [<at 0 0.1 RCIM_Cubic>,<at 0.209 0 RCIM_Cubic>]}
        }})
      Evening Color=&ColorFromCurve(ColorCurve=DataInterface() {propertyValues {
         AlphaCurve {keys []}
         BlueCurve {keys [{arriveTangent=-331.6383, interpMode=RCIM_Cubic, time=0, value=0.1}, <at 0.209 0 RCIM_Cubic>]}
         GreenCurve {keys [{arriveTangent=-87.3695, interpMode=RCIM_Cubic, time=0, value=0.1}, <at 0.209 0 RCIM_Cubic>]}
         RedCurve {keys [<at 0 0.1 RCIM_Cubic>,<at 0.209 0 RCIM_Cubic>]}
        }})
     ))
    ScaleA=true
    ScaleRGB=true
   )
   &SubUVAnimation(
    End Frame Range Override=30
    "Sprite Renderer "=DataInterface() {propertyValues {SpriteRenderer=&self:SmokeBody02_6.NiagaraSpriteRendererProperties_0}}
    Start Frame Range Override=&RandomRangeInt(Minimum=0, Maximum=5)
    UseEndFrame=true
    UseStartFrame=true
   ) {module=SubUVAnimation001}
  ]
  rendererClasses [&NiagaraSpriteRendererProperties]
  renderers [{rendererClass=&NiagaraSpriteRendererProperties}]
 }
 SmokeBurst {
  emitterUpdateScript [
   &EmitterState(Life Cycle Mode=::NewEnumerator1, Loop Behavior=::NewEnumerator1, Loop Delay=1.2, Loop Duration=1, UseLoopDelay=true)
   &SpawnBurst_Instantaneous(Spawn Count=25, Use Spawn Probability=false, UseLoopCountLimit=false)
  ]
  particleSpawnScript [
   &InitializeParticle(
    Color=&MultiplyLinearColor(Linear Color=<rgba 0.7031 0.7031 0.7031 0.8>, Scale Factor=&MakeVector4_FromFloat(Float=1))
    Lifetime Max=1.3
    Lifetime Min=0.7
    Lifetime Mode=::NewEnumerator1
    Sprite Rotation Mode=::NewEnumerator2
    Sprite Size Mode=::NewEnumerator1
    Uniform Sprite Size Max=120
    Uniform Sprite Size Min=100
    UsePositionOffset=true
    Write Material Random=true
   )
   &AddVelocityInCone(
    Cone Angle=80
    Cone Axis=<0 0 -1>
    Module.Use Velocity Falloff On Cone Axis=false
    Velocity Strength=&RandomRangeFloat(Minimum=100, Maximum=500)
   ) {module=AddVelocityInCone001}
   &CameraOffset(Camera Offset Amount=-20)
   &ShapeLocation(MaskSpawnGroup=false, Override Seed=false, Sphere Radius=20, UseOwnerScale=true)
  ]
  particleUpdateScript [
   &Drag(Drag=&RandomRangeFloat(Minimum=1, Maximum=1), Use Linear Drag=true, Use Rotational Drag=true)
   &InheritSourceMovement(Applied Position Scale=<0.5 0.5 0.5>, ApplyForce=false, ApplyPosition=true, ApplyPositionFalloff=false)
   &SolveForcesAndVelocity(Clamp Velocity=false, Limit Acceleration=false)
   &ParticleState()
   &SpriteRotationRate(Rotation Rate=50)
   &ScaleSpriteSize(Uniform Curve Sprite Scale=DataInterface() {propertyValues {Curve {keys [{interpMode=RCIM_Cubic, tangentMode=RCTM_User, time=0, value=1.2}, {interpMode=RCIM_Cubic, tangentMode=RCTM_User, time=1, value=1.5}]}}})
   &DynamicMaterialParameters(
    Index 0 Param 1=&RandomRangeFloat(Minimum=0, Maximum=1)
    Module.Index0Param0Write=true
    Module.Index0Param1Write=true
    Module.Index0Param2Write=true
    Module.Index0Param3Write=true
   )
   &ScaleColor(
    Scale Alpha=&Multiply_Float(
     A=&FloatFromCurve(
      FloatCurve=DataInterface() {propertyValues {Curve {keys [
          {arriveTangent=10.2356, interpMode=RCIM_Cubic, leaveTangent=10.2356, tangentMode=RCTM_User, time=0, value=0}
          <at 0.3 1>
          <at 0.7 0.4>
          <at 1 0>
         ]}}}
      Scale Curve=&RandomRangeFloat(Minimum=0.5, Maximum=3)
     )
     B=1
    )
    Scale RGB=&MakeVectorFromLinearColorRGB(Color=&FNTimeOfDayColor(
      Night Color=&ColorFromCurve(ColorCurve=DataInterface() {propertyValues {
         AlphaCurve {keys []}
         BlueCurve {keys [{arriveTangent=-331.6383, interpMode=RCIM_Cubic, time=0, value=0.15}, <at 0.4083 0 RCIM_Cubic>]}
         GreenCurve {keys [{arriveTangent=-87.3695, interpMode=RCIM_Cubic, time=0, value=0.0712}, <at 0.4083 0 RCIM_Cubic>]}
         RedCurve {keys [<at 0 0.045 RCIM_Cubic>,<at 0.4083 0 RCIM_Cubic>]}
        }})
      Morning Color=&ColorFromCurve(ColorCurve=DataInterface() {propertyValues {
         AlphaCurve {keys []}
         BlueCurve {keys [{arriveTangent=-331.6383, interpMode=RCIM_Cubic, time=0, value=0.15}, <at 0.4083 0 RCIM_Cubic>]}
         GreenCurve {keys [{arriveTangent=-87.3695, interpMode=RCIM_Cubic, time=0, value=0.15}, <at 0.4083 0 RCIM_Cubic>]}
         RedCurve {keys [<at 0 0.15 RCIM_Cubic>,<at 0.4083 0 RCIM_Cubic>]}
        }})
      Day Color=&ColorFromCurve(ColorCurve=DataInterface() {propertyValues {
         AlphaCurve {keys []}
         BlueCurve {keys [{arriveTangent=-331.6383, interpMode=RCIM_Cubic, time=0, value=0.15}, <at 0.4083 0 RCIM_Cubic>]}
         GreenCurve {keys [{arriveTangent=-87.3695, interpMode=RCIM_Cubic, time=0, value=0.15}, <at 0.4083 0 RCIM_Cubic>]}
         RedCurve {keys [<at 0 0.15 RCIM_Cubic>,<at 0.4083 0 RCIM_Cubic>]}
        }})
      Evening Color=&ColorFromCurve(ColorCurve=DataInterface() {propertyValues {
         AlphaCurve {keys []}
         BlueCurve {keys [{arriveTangent=-331.6383, interpMode=RCIM_Cubic, time=0, value=0.15}, <at 0.4083 0 RCIM_Cubic>]}
         GreenCurve {keys [{arriveTangent=-87.3695, interpMode=RCIM_Cubic, time=0, value=0.15}, <at 0.4083 0 RCIM_Cubic>]}
         RedCurve {keys [<at 0 0.15 RCIM_Cubic>,<at 0.4083 0 RCIM_Cubic>]}
        }})
     ))
    ScaleA=true
    ScaleRGB=true
   )
   &SubUVAnimation(
    End Frame Range Override=3
    Evaluation Type=::NewEnumerator0
    "Sprite Renderer "=DataInterface() {propertyValues {SpriteRenderer=&self:SmokeBurst_8.NiagaraSpriteRendererProperties_0}}
    SubUV Animation Mode:ENiagara_SubUVLookupModeV2=::NewEnumerator2
    UseEndFrame=true
    UseStartFrame=true
   )
  ]
  rendererClasses [&NiagaraSpriteRendererProperties]
  renderers [{rendererClass=&NiagaraSpriteRendererProperties}]
 }
 SmokeDetail {
  emitterUpdateScript [
   &EmitterState(Life Cycle Mode=::NewEnumerator1, Loop Behavior=::NewEnumerator1, Loop Duration=1.2, UseLoopDelay=false)
   &SpawnRate(SpawnRate=13, Use Spawn Probability=true)
  ]
  particleSpawnScript [
   &PartitionParticles(Number Of Partitions=5)
   &InitializeParticle(
    Color=&MultiplyLinearColor(Linear Color=<rgba 0.7031 0.7031 0.7031 0.8>, Scale Factor=&MakeVector4_FromFloat(Float=1))
    Lifetime Max=1.6
    Lifetime Min=1.5
    Lifetime Mode=::NewEnumerator1
    Position Offset=<0 0 50>
    Sprite Rotation Mode=::NewEnumerator2
    Sprite Size Mode=::NewEnumerator1
    Uniform Sprite Size Max=100
    Uniform Sprite Size Min=100
    UsePositionOffset=true
    Write Material Random=true
   )
   &ShapeLocation(MaskSpawnGroup=false, Override Seed=false, Sphere Radius=5, UseOwnerScale=true)
   &SkeletalMeshLocation(
    Bone / Socket Index=$Particles.PartitionParticles.Partition
    Bone Sampling Mode=::NewEnumerator4
    MaskSpawnGroup=false
    Skeletal Mesh=$System.Skeletal Mesh
   )
  ]
  particleUpdateScript [
   &Drag(Use Linear Drag=true, Use Rotational Drag=true)
   &SolveForcesAndVelocity(Clamp Velocity=false, Limit Acceleration=false)
   &ParticleState()
   &SpriteRotationRate(Rotation Rate=&RandomRangeFloat(Minimum=-50, Maximum=50))
   &ScaleSpriteSize(Uniform Curve Sprite Scale=DataInterface() {propertyValues {Curve {keys [{interpMode=RCIM_Cubic, tangentMode=RCTM_User, time=0, value=0.7}, {interpMode=RCIM_Cubic, tangentMode=RCTM_User, time=1, value=1}]}}})
   &ScaleColor(
    Scale Alpha=&Multiply_Float(
     A=&FloatFromCurve(FloatCurve=DataInterface() {propertyValues {Curve {keys [<at 0 1>,<at 1.1 1>,<at 1.6 0>]}}}, CurveIndex=$Emitter.Age, Scale Curve=1.3)
     B=&FloatFromCurve(FloatCurve=DataInterface() {propertyValues {Curve {keys [
          {interpMode=RCIM_Cubic, tangentMode=RCTM_User, time=0, value=0}
          {interpMode=RCIM_Cubic, tangentMode=RCTM_User, time=0.1, value=1}
          {arriveTangent=-1.1111, interpMode=RCIM_Cubic, leaveTangent=-1.1111, time=0.8, value=1}
          {tangentMode=RCTM_User, time=1, value=0}
         ]}}})
    )
    Scale RGB=&MakeCustomVectorFromBool(
     Bool=$User.bIsFrontEndPreview
     True Vector=&MakeVectorFromLinearColorRGB(Color=&ColorFromCurve(ColorCurve=DataInterface() {propertyValues {
         AlphaCurve {keys []}
         BlueCurve {keys [{arriveTangent=-331.6383, interpMode=RCIM_Cubic, time=0, value=0.0781}, <at 0.7598 0 RCIM_Cubic>]}
         GreenCurve {keys [{arriveTangent=-87.3695, interpMode=RCIM_Cubic, time=0, value=0.0609}, <at 0.7598 0 RCIM_Cubic>]}
         RedCurve {keys [<at 0 0.0469 RCIM_Cubic>,<at 0.7598 0 RCIM_Cubic>]}
        }}))
     False Vector=&MakeVectorFromLinearColorRGB(Color=&FNTimeOfDayColor(
       Night Color=&ColorFromCurve(ColorCurve=DataInterface() {propertyValues {
          AlphaCurve {keys []}
          BlueCurve {keys [{arriveTangent=-331.6383, interpMode=RCIM_Cubic, time=0, value=0.0781}, <at 0.7598 0 RCIM_Cubic>]}
          GreenCurve {keys [{arriveTangent=-87.3695, interpMode=RCIM_Cubic, time=0, value=0.0371}, <at 0.7598 0 RCIM_Cubic>]}
          RedCurve {keys [<at 0 0.0234 RCIM_Cubic>,<at 0.7598 0 RCIM_Cubic>]}
         }})
       Morning Color=&ColorFromCurve(ColorCurve=DataInterface() {propertyValues {
          AlphaCurve {keys []}
          BlueCurve {keys [{arriveTangent=-331.6383, interpMode=RCIM_Cubic, time=0, value=0.0781}, <at 0.7598 0 RCIM_Cubic>]}
          GreenCurve {keys [{arriveTangent=-87.3695, interpMode=RCIM_Cubic, time=0, value=0.0781}, <at 0.7598 0 RCIM_Cubic>]}
          RedCurve {keys [<at 0 0.0781 RCIM_Cubic>,<at 0.7598 0 RCIM_Cubic>]}
         }})
       Day Color=&ColorFromCurve(ColorCurve=DataInterface() {propertyValues {
          AlphaCurve {keys []}
          BlueCurve {keys [{arriveTangent=-331.6383, interpMode=RCIM_Cubic, time=0, value=0.0781}, <at 0.7598 0 RCIM_Cubic>]}
          GreenCurve {keys [{arriveTangent=-87.3695, interpMode=RCIM_Cubic, time=0, value=0.0781}, <at 0.7598 0 RCIM_Cubic>]}
          RedCurve {keys [<at 0 0.0781 RCIM_Cubic>,<at 0.7598 0 RCIM_Cubic>]}
         }})
       Evening Color=&ColorFromCurve(ColorCurve=DataInterface() {propertyValues {
          AlphaCurve {keys []}
          BlueCurve {keys [{arriveTangent=-331.6383, interpMode=RCIM_Cubic, time=0, value=0.0781}, <at 0.7598 0 RCIM_Cubic>]}
          GreenCurve {keys [{arriveTangent=-87.3695, interpMode=RCIM_Cubic, time=0, value=0.0781}, <at 0.7598 0 RCIM_Cubic>]}
          RedCurve {keys [<at 0 0.0781 RCIM_Cubic>,<at 0.7598 0 RCIM_Cubic>]}
         }})
      ))
    )
    ScaleA=true
    ScaleRGB=true
   )
   &SubUVAnimation(
    End Frame Range Override=30
    "Sprite Renderer "=DataInterface() {propertyValues {SpriteRenderer=&self:SmokeDetail_1.NiagaraSpriteRendererProperties_0}}
    Start Frame Range Override=&RandomRangeInt(Minimum=0, Maximum=5)
    UseEndFrame=true
    UseStartFrame=true
   )
  ]
  rendererClasses [&NiagaraSpriteRendererProperties]
  renderers [{rendererClass=&NiagaraSpriteRendererProperties}]
 }
 Smoke_Body {
  emitterUpdateScript [
   &EmitterState(Life Cycle Mode=::NewEnumerator1, Loop Behavior=::NewEnumerator1, Loop Delay=0.2, Loop Duration=1, UseLoopDelay=true)
   &SpawnPerUnit(Module.Use Max Movement Threshold=true, Module.Use Movement Tolerance=false, Spawn Spacing=30, Use Spawn Probability=false)
   &SpawnRate(SpawnRate=20, Use Spawn Probability=true)
  ]
  particleSpawnScript [
   &PartitionParticles(Number Of Partitions=1)
   &InitializeParticle(
    Color=&MultiplyLinearColor(Linear Color=<rgba 0.7031 0.7031 0.7031 0.8>, Scale Factor=&MakeVector4_FromFloat(Float=1))
    Lifetime Max=1.5
    Lifetime Min=0.7
    Lifetime Mode=::NewEnumerator1
    Sprite Rotation Mode=::NewEnumerator2
    Sprite Size Mode=::NewEnumerator1
    Uniform Sprite Size Max=70
    Uniform Sprite Size Min=50
    UsePositionOffset=true
    Write Material Random=true
   )
   &CameraOffset(Camera Offset Amount=-20)
   &ShapeLocation(MaskSpawnGroup=false, Override Seed=false, Sphere Radius=10, UseOwnerScale=true)
  ]
  particleUpdateScript [
   &SolveForcesAndVelocity(Clamp Velocity=false, Limit Acceleration=false)
   &ParticleState()
   &SpriteRotationRate(Rotation Rate=50)
   &ScaleSpriteSize(Uniform Curve Sprite Scale=DataInterface() {propertyValues {Curve {keys [
        {arriveTangent=10.8988, interpMode=RCIM_Cubic, leaveTangent=10.8988, tangentMode=RCTM_User, time=0, value=0}
        {arriveTangent=-1.0121, interpMode=RCIM_Cubic, leaveTangent=-1.0121, tangentMode=RCTM_User, time=0.2831, value=1.5381}
        {arriveTangent=-1.169, interpMode=RCIM_Cubic, leaveTangent=-1.169, time=0.5, value=0.875}
        <at 1 0.7>
       ]}}})
   &DynamicMaterialParameters(
    Index 0 Param 1=&RandomRangeFloat(Minimum=0, Maximum=1)
    Module.Index0Param0Write=true
    Module.Index0Param1Write=true
    Module.Index0Param2Write=true
    Module.Index0Param3Write=true
   )
   &ScaleColor(
    Scale Alpha=&Multiply_Float(
     A=&FloatFromCurve(
      FloatCurve=DataInterface() {propertyValues {Curve {keys [
          {arriveTangent=10.2356, interpMode=RCIM_Cubic, leaveTangent=10.2356, tangentMode=RCTM_User, time=0, value=0}
          <at 0.3 1>
          <at 0.7 0.4>
          <at 1 0>
         ]}}}
      Scale Curve=&RandomRangeFloat(Minimum=0.5, Maximum=3)
     )
     B=&FloatFromCurve(FloatCurve=DataInterface() {propertyValues {Curve {keys [<at 0 1>,<at 1.1 1>,<at 1.6 0>]}}}, CurveIndex=$Emitter.Age)
    )
    Scale RGB=&MakeCustomVectorFromBool(
     Bool=$User.bIsFrontEndPreview
     True Vector=&MakeVectorFromLinearColorRGB(Color=&ColorFromCurve(ColorCurve=DataInterface() {propertyValues {
         AlphaCurve {keys []}
         BlueCurve {keys [{arriveTangent=-331.6383, interpMode=RCIM_Cubic, time=0, value=0.0781}, <at 0.209 0 RCIM_Cubic>]}
         GreenCurve {keys [{arriveTangent=-87.3695, interpMode=RCIM_Cubic, time=0, value=0.0609}, <at 0.209 0 RCIM_Cubic>]}
         RedCurve {keys [<at 0 0.0469 RCIM_Cubic>,<at 0.209 0 RCIM_Cubic>]}
        }}))
     False Vector=&MakeVectorFromLinearColorRGB(Color=&FNTimeOfDayColor(
       Night Color=&ColorFromCurve(ColorCurve=DataInterface() {propertyValues {
          AlphaCurve {keys []}
          BlueCurve {keys [{arriveTangent=-331.6383, interpMode=RCIM_Cubic, time=0, value=0.1}, <at 0.209 0 RCIM_Cubic>]}
          GreenCurve {keys [{arriveTangent=-87.3695, interpMode=RCIM_Cubic, time=0, value=0.055}, <at 0.209 0 RCIM_Cubic>]}
          RedCurve {keys [<at 0 0.04 RCIM_Cubic>,<at 0.209 0 RCIM_Cubic>]}
         }})
       Morning Color=&ColorFromCurve(ColorCurve=DataInterface() {propertyValues {
          AlphaCurve {keys []}
          BlueCurve {keys [{arriveTangent=-331.6383, interpMode=RCIM_Cubic, time=0, value=0.0781}, <at 0.209 0 RCIM_Cubic>]}
          GreenCurve {keys [{arriveTangent=-87.3695, interpMode=RCIM_Cubic, time=0, value=0.0781}, <at 0.209 0 RCIM_Cubic>]}
          RedCurve {keys [<at 0 0.0781 RCIM_Cubic>,<at 0.209 0 RCIM_Cubic>]}
         }})
       Day Color=&ColorFromCurve(ColorCurve=DataInterface() {propertyValues {
          AlphaCurve {keys []}
          BlueCurve {keys [{arriveTangent=-331.6383, interpMode=RCIM_Cubic, time=0, value=0.0781}, <at 0.209 0 RCIM_Cubic>]}
          GreenCurve {keys [{arriveTangent=-87.3695, interpMode=RCIM_Cubic, time=0, value=0.0781}, <at 0.209 0 RCIM_Cubic>]}
          RedCurve {keys [<at 0 0.0781 RCIM_Cubic>,<at 0.209 0 RCIM_Cubic>]}
         }})
       Evening Color=&ColorFromCurve(ColorCurve=DataInterface() {propertyValues {
          AlphaCurve {keys []}
          BlueCurve {keys [{arriveTangent=-331.6383, interpMode=RCIM_Cubic, time=0, value=0.0781}, <at 0.209 0 RCIM_Cubic>]}
          GreenCurve {keys [{arriveTangent=-87.3695, interpMode=RCIM_Cubic, time=0, value=0.0781}, <at 0.209 0 RCIM_Cubic>]}
          RedCurve {keys [<at 0 0.0781 RCIM_Cubic>,<at 0.209 0 RCIM_Cubic>]}
         }})
      ))
    )
    ScaleA=true
    ScaleRGB=true
   )
   &SubUVAnimation(
    End Frame Range Override=3
    Evaluation Type=::NewEnumerator0
    "Sprite Renderer "=DataInterface() {propertyValues {SpriteRenderer=&self:Smoke_Body_5.NiagaraSpriteRendererProperties_0}}
    SubUV Animation Mode:ENiagara_SubUVLookupModeV2=::NewEnumerator2
    UseEndFrame=true
    UseStartFrame=true
   )
  ]
  rendererClasses [&NiagaraSpriteRendererProperties]
  renderers [{rendererClass=&NiagaraSpriteRendererProperties}]
 }
 Source {
  emitterUpdateScript [
   &EmitterState(Life Cycle Mode=::NewEnumerator1, Loop Behavior=::NewEnumerator1, Loop Delay=0.2, Loop Duration=1, UseLoopDelay=true)
   &SpawnBurst_Instantaneous(Spawn Count=3, Use Spawn Probability=false, UseLoopCountLimit=false)
  ]
  particleSpawnScript [
   &PartitionParticles(Number Of Partitions=3)
   &InitializeParticle(
    Color=&MultiplyLinearColor(Linear Color=<rgba 0.7031 0.7031 0.7031 0.8>, Scale Factor=&MakeVector4_FromFloat(Float=1))
    Lifetime Max=1.5
    Lifetime Min=0.7
    Lifetime Mode=::NewEnumerator1
    Sprite Rotation Mode=::NewEnumerator2
    Sprite Size Mode=::NewEnumerator3
    UsePositionOffset=true
    Write Material Random=true
   )
   &ShapeLocation(
    MaskSpawnGroup=false
    Override Seed=false
    Shape Origin=&VectorFromFloat(Value=0)
    Sphere Radius=50
    UseOwnerScale=true
   )
  ]
  particleUpdateScript [
   &SolveForcesAndVelocity(Clamp Velocity=false, Limit Acceleration=false)
   &ParticleState()
   &SkeletalMeshLocation(
    Bone / Socket Index=12
    Bone Sampling Mode=::NewEnumerator4
    MaskSpawnGroup=false
    Sampled Position Offset=&AddVectorToPosition(Position=$Particles.Initial.ShapeLocation.ShapePosition, Vector=<0 0 0>)
    Skeletal Mesh=$System.Skeletal Mesh
   )
  ]
  rendererClasses [&NiagaraSpriteRendererProperties]
  renderers [{rendererClass=&NiagaraSpriteRendererProperties}]
 }
}
summary {
 systemName=NS_Emote_MistMylar_deploy
 emitters [
  {emitterName=SmokeDetail, rendererClasses [&NiagaraSpriteRendererProperties]}
  {emitterName=Smoke_Body, rendererClasses [&NiagaraSpriteRendererProperties]}
  {emitterName=SmokeBlur, rendererClasses [&NiagaraRibbonRendererProperties]}
  {emitterName=Flash, rendererClasses [&NiagaraSpriteRendererProperties]}
  {emitterName=SmokeBody02, rendererClasses [&NiagaraSpriteRendererProperties]}
  {emitterName=Source, rendererClasses [&NiagaraSpriteRendererProperties]}
  {emitterName=BodyTrail, rendererClasses [&NiagaraRibbonRendererProperties]}
  {emitterName=SmokeBurst, rendererClasses [&NiagaraSpriteRendererProperties]}
  {bEnabled=false, emitterName=Cloud009, rendererClasses [&NiagaraSpriteRendererProperties]}
 ]
 userVariables [
  User.bIsFrontEnd {type=NiagaraBool}
  User.bIsFrontEndPreview {type=NiagaraBool}
 ]
}
